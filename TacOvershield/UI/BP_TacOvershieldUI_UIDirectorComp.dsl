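; UI director component that adds and removes the Tactical Overshield HUD scenes for the owning local player.
; On UI ready and each HUD element visibility refresh, it shows the main scene unless the minigame's visibility mutator supplies a player info widget override.
; A cached flag avoids redundant scene changes.
; A gameplay event router listener adds the post-match scene when the local player's place changes, and EndPlay releases both subscriptions.
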(event EventOnReadyToShowUI
  (CallFunction|UpdateTacticalOverShieldScene)
  (bind _asfort_player_controller_gameplay (Utilities|Casting|CastToFortPlayerControllerGameplay (CallFunction|GetOwningLocalPlayerController)))
  (HUD|BindEventtoOnHUDElementVisibilityRefreshedDelegate _asfort_player_controller_gameplay (EventDispatchers|CreateEvent :!event UpdateTacticalOverShieldScene))
  (bind _asynctaskproxy (GameplayEventRouter|StartListeningToEvent (GameplayEventRouter|GetGlobalEventRouter) "/Script/FortniteGame.OnLocalPlayerPlaceChangedEvent" (CallFunction|GetOwningLocalPlayerController))
    (:then
      (Variables|Default|SetPostMatchSceneEvent _asynctaskproxy))
    (:OnEventReceived
      (CallFunction|AddScene "/TacOvershield/UI/UIScene_TacOverShield_PostMatch.UIScene_TacOverShield_PostMatch"))))

(event EventEndPlay
  (params (EndPlayReason :type EEndPlayReason%20Enum :default "Destroyed"))
  (bind _asfort_player_controller_gameplay (Utilities|Casting|CastToFortPlayerControllerGameplay (CallFunction|GetOwningLocalPlayerController)))
  (HUD|UnbindEventfromOnHUDElementVisibilityRefreshedDelegate _asfort_player_controller_gameplay (EventDispatchers|CreateEvent :!event UpdateTacticalOverShieldScene))
  (GameplayEventRouter|StopListeningtoEvent (Variables|Default|GetPostMatchSceneEvent)))

(fn DoesPlayerInfoWidgetOverrideExist
  (Utilities|FlowControl|Sequence
    (:then_0
      (bind _asfort_player_controller_gameplay (Utilities|Casting|CastToFortPlayerControllerGameplay (CallFunction|GetOwningLocalPlayerController)))
      (Utilities|IsValid (Creative|GetMinigame _asfort_player_controller_gameplay)
        (:Is%20Valid
          (Utilities|IsValid (Actor|GetComponentbyClass (Creative|GetMinigame _asfort_player_controller_gameplay) "/Script/FortniteGame.FortMutatorListComponent")
            (:Is%20Valid
              (bind _asfort_athena_mutator_hudelement_visibility (Utilities|Casting|CastToFortAthenaMutator_HUDElementVisibility (Modifier|GetMutatorbyClass (Actor|GetComponentbyClass (Creative|GetMinigame _asfort_player_controller_gameplay) "/Script/FortniteGame.FortMutatorListComponent") "/Script/FortniteGame.FortAthenaMutator_HUDElementVisibility")))
              (bind _returnvalue (Class|FortAthenaMutatorHUDElementVisibility|GetPlayerInfoWidgetOverrideClass _asfort_athena_mutator_hudelement_visibility (CallFunction|GetOwningLocalPlayerController)))
              (return (Utilities|IsValidSoftClassReference _returnvalue)))))))
    (:then_1
      (return false))))

(fn UpdateTacticalOverShieldScene
  ; Don't add scene if an override exists
  (LocalVariables|SetbShouldAddScene (not (CallFunction|DoesPlayerInfoWidgetOverrideExist)))
  (if (!= (Variables|Default|GetbSceneAdded) (LocalVariables|GetbShouldAddScene))
    (bind _output_get (Variables|Default|SetbSceneAdded (LocalVariables|GetbShouldAddScene)))
    (if _output_get
      (CallFunction|AddScene "/TacOvershield/UI/TacOverShieldScene.TacOverShieldScene")
      (else
        (CallFunction|RemoveScene "/TacOvershield/UI/TacOverShieldScene.TacOverShieldScene")))))
