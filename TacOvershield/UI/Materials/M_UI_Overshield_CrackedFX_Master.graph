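# Translucent UI master material for the Overshield cracked effect, animating shard shapes from a packed texture.
# Each RGB channel scales around its center, then gets subtract, multiply and alpha adjustments lerped between per-channel start and stop values.
# Shared Animate parameters drive these lerps, with no time node in the graph.
# The summed, saturated mask times Animate_Color blends ColorStop toward ColorStart.
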Material /TacOvershield/UI/Materials/M_UI_Overshield_CrackedFX_Master
graph {
 properties {blendMode=BLEND_Translucent, materialDomain=MD_UI, pixelDepthOffsetMode=PDOM_Legacy, usedShadingModels=Default Lit}
 nodes [
  n1 = LinearInterpolate(
   A=ScalarParameter() {defaultValue=0.5, group=Red, parameterName=RedChannel_Start_Scale}
   B=ScalarParameter() {defaultValue=2, group=Red, parameterName=RedChannel_Stop_Scale}
   Alpha=ScalarParameter() {group=Animate, parameterName=Animate_Scale}
  )
  n2 = Subtract(
   A=TextureSample(UVs=MaterialFunctionCall(Texture Scale=@n1) {materialFunction=&ScaleUVsByCenter}.UVs) {samplerType=SAMPLERTYPE_LinearColor, texture=/TacOvershield/UI/Textures/T_UI_Overshield_Shards_128x}.R
   B=LinearInterpolate(
    A=ScalarParameter() {group=Red, parameterName=RedChannel_Subtract_Start}
    B=ScalarParameter() {group=Red, parameterName=RedChannel_Subtract_Stop}
    Alpha=ScalarParameter() {group=Animate, parameterName=Animate_Subtract}
   )
  )
  n3 = LinearInterpolate(
   A=ScalarParameter() {defaultValue=1, group=Red, parameterName=RedChannel_Multiply_Start}
   B=ScalarParameter() {group=Red, parameterName=RedChannel_Multiply_Stop}
   Alpha=ScalarParameter() {group=Animate, parameterName=Animate_Multiply}
  )
  n4 = Multiply(
   A=Saturate(None=Multiply(A=@n2, B=@n3))
   B=LinearInterpolate(
    A=ScalarParameter() {defaultValue=1, group=Red, parameterName=RedChannel_Alpha_Start}
    B=ScalarParameter() {group=Red, parameterName=RedChannel_Alpha_Stop}
    Alpha=ScalarParameter() {group=Animate, parameterName=Animate_Alpha}
   )
  )
  n5 = LinearInterpolate(
   A=ScalarParameter() {defaultValue=0.5, group=Green, parameterName=GreenChannel_Start_Scale}
   B=ScalarParameter() {defaultValue=2, group=Green, parameterName=GreenChannel_Stop_Scale}
   Alpha=ScalarParameter() {group=Animate, parameterName=Animate_Scale}
  )
  n6 = Subtract(
   A=TextureSample(UVs=MaterialFunctionCall(Texture Scale=@n5) {materialFunction=&ScaleUVsByCenter}.UVs) {samplerType=SAMPLERTYPE_LinearColor, texture=&T_UI_Overshield_Shards_128x}.G
   B=LinearInterpolate(
    A=ScalarParameter() {group=Green, parameterName=GreenChannel_Subtract_Start}
    B=ScalarParameter() {group=Green, parameterName=GreenChannel_Subtract_Stop}
    Alpha=ScalarParameter() {group=Animate, parameterName=Animate_Subtract}
   )
  )
  n7 = LinearInterpolate(
   A=ScalarParameter() {defaultValue=1, group=Green, parameterName=GreenChannel_Multiply_Start}
   B=ScalarParameter() {group=Green, parameterName=GreenChannel_Multiply_Stop}
   Alpha=ScalarParameter() {group=Animate, parameterName=Animate_Multiply}
  )
  n8 = Multiply(
   A=Saturate(None=Multiply(A=@n6, B=@n7))
   B=LinearInterpolate(
    A=ScalarParameter() {defaultValue=1, group=Green, parameterName=GreenChannel_Alpha_Start}
    B=ScalarParameter() {group=Green, parameterName=GreenChannel_Alpha_Stop}
    Alpha=ScalarParameter() {group=Animate, parameterName=Animate_Alpha}
   )
  )
  n9 = ScalarParameter() {defaultValue=0.5, group=Blue, parameterName=BlueChannel_Start_Scale}
  n10 = ScalarParameter() {defaultValue=2, group=Blue, parameterName=BlueChannel_Stop_Scale}
  n11 = ScalarParameter() {group=Animate, parameterName=Animate_Scale}
  n12 = TextureSample(UVs=MaterialFunctionCall(Texture Scale=LinearInterpolate(A=@n9, B=@n10, Alpha=@n11)) {materialFunction=&ScaleUVsByCenter}.UVs) {samplerType=SAMPLERTYPE_LinearColor, texture=&T_UI_Overshield_Shards_128x}
  n13 = LinearInterpolate(
   A=ScalarParameter() {group=Blue, parameterName=BlueChannel_Subtract_Start}
   B=ScalarParameter() {group=Blue, parameterName=BlueChannel_Subtract_Stop}
   Alpha=ScalarParameter() {group=Animate, parameterName=Animate_Subtract}
  )
  n14 = ScalarParameter() {defaultValue=1, group=Blue, parameterName=BlueChannel_Multiply_Start}
  n15 = ScalarParameter() {group=Blue, parameterName=BlueChannel_Multiply_Stop}
  n16 = ScalarParameter() {group=Animate, parameterName=Animate_Multiply}
  n17 = Saturate(None=Multiply(A=Subtract(A=@n12.B, B=@n13), B=LinearInterpolate(A=@n14, B=@n15, Alpha=@n16)))
  n18 = LinearInterpolate(
   A=ScalarParameter() {defaultValue=1, group=Blue, parameterName=BlueChannel_Alpha_Start}
   B=ScalarParameter() {group=Blue, parameterName=BlueChannel_Alpha_Stop}
   Alpha=ScalarParameter() {group=Animate, parameterName=Animate_Alpha}
  )
  n19 = Add(A=Add(A=Saturate(None=@n4), B=Saturate(None=@n8)), B=Saturate(None=Multiply(A=@n17, B=@n18)))
  LinearInterpolate(
   A=VectorParameter() {parameterName=ColorStop, defaultValue=<rgba 1 0 0 1>}.RGB
   B=VectorParameter() {parameterName=ColorStart, defaultValue=<rgba 0 1 0 1>}.RGB
   Alpha=Multiply(A=Saturate(None=@n19), B=ScalarParameter() {defaultValue=1, parameterName=Animate_Color})
  )
 ]
}
