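# Translucent UI master material for the Tactical Overshield backing, with static switches for pattern, outline, swipe and plain backing looks.
# The pattern pans a tile grid, optionally in screen space, using per-tile cosine pseudo-randomness to offset and animate glow, solid and outline layers from one texture's RGB channels.
# Rounded-square SmoothStep masks shape backing and outline, while parameters drive the color gradient and rotated sine swipe.
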Material /TacOvershield/UI/Materials/M_UI_TacOvershield_Backing_Master
graph {
 properties {blendMode=BLEND_Translucent, materialDomain=MD_UI, pixelDepthOffsetMode=PDOM_Legacy, usedShadingModels=Default Lit}
 nodes [
  n1 = Multiply(A=NamedRerouteUsage(), B=ScalarParameter() {defaultValue=1, group=Overall, parameterName=Amount_GlowTiles})
  n2 = Multiply(A=NamedRerouteUsage(), B=ScalarParameter() {defaultValue=1, group=Overall, parameterName=Amount_SolidTiles})
  n3 = NamedRerouteUsage()
  n4 = ScalarParameter() {defaultValue=1, group=Overall, parameterName=Amount_OutlineTiles}
  n5 = Saturate(None=Add(A=Add(A=@n1, B=@n2), B=Multiply(A=@n3, B=@n4)))
  n6 = ScalarParameter() {defaultValue=2, group=Overall, parameterName=Pattern_Inensity}
  n7 = DotProduct(A=Saturate(None=Multiply(A=@n5, B=@n6)), B=NamedRerouteUsage())
  NamedRerouteDeclaration(None=LinearInterpolate(
    A=VectorParameter() {group=Overall, parameterName=Color_2, defaultValue=<rgba 0 0.1667 1 1>}.RGB
    B=VectorParameter() {group=Overall, parameterName=Color_1, defaultValue=<rgba 0 1 1 1>}.RGB
    Alpha=@n7
   )) {name=Pattern, nodeColor=<rgba 0 1 0.5313 1>}
  n8 = Saturate(None=MaterialFunctionCall!(
    UV
    Size=ScalarParameter() {defaultValue=0.4, group=Backing, parameterName=Size_2}
    Curvature=ScalarParameter() {defaultValue=0.75, group=Backing, parameterName=CornerRadius_2}
   ) {materialFunction=/Game/Effects/Fort_Effects/MaterialFunctions/MF_RoundedSquare}.Linear Gradient)
  n9 = ScalarParameter() {defaultValue=0.05, group=Backing, parameterName=Min_2}
  n10 = ScalarParameter() {defaultValue=1, group=Backing, parameterName=Max_2}
  n11 = Multiply(
   A=Saturate(None=MaterialFunctionCall(Alpha=@n8, Min=@n9, Max=@n10) {materialFunction=&SmoothStep}.Result)
   B=ScalarParameter() {defaultValue=1, group=Backing, parameterName=EdgeHardness_2}
  )
  NamedRerouteDeclaration(None=OneMinus(None=Saturate(None=@n11))) {name=Backing_1, nodeColor=<rgba 0.1484 1 0 1>}
  n12 = ScalarParameter() {defaultValue=0.125, group=Tiling, parameterName=MoreConsistentTiles}
  n13 = TextureCoordinate() {coordinateIndex=4}
  n14 = MaterialFunctionCall() {materialFunction=/FortUILibrary/Foundation/Art_ToolKit/Tools/MF_UI_SquareToScreenSpace}
  n15 = ScalarParameter() {defaultValue=1, parameterName=UseScreenSpace}
  n16 = Constant() {r=1}
  n17 = ScalarParameter() {defaultValue=1, group=Overall, parameterName=StetchV}
  n18 = Panner(
   Coordinate=Multiply(A=LinearInterpolate(A=@n13, B=@n14.Y_Priority, Alpha=@n15), B=AppendVector(A=@n16, B=@n17))
   Speed=AppendVector(
    A=ScalarParameter() {group=Tiling, parameterName=PanSpeed_SolidTiles_X}
    B=ScalarParameter() {group=Tiling, parameterName=PanSpeed_SolidTiles_Y}
   )
  ) {speedX=0.5}
  n19 = Frac(None=@n18)
  n20 = ScalarParameter() {defaultValue=16, group=Tiling, parameterName=Tile}
  n21 = MaterialFunctionCall(Vector 2=Ceil(None=Multiply(A=@n19, B=@n20))) {materialFunction=/Engine/Functions/Engine_MaterialFunctions02/Utility/VectorLength}
  n22 = Multiply(A=@n21.V2 Length, B=Multiply(A=Time(), B=ScalarParameter() {defaultValue=0.125, group=Tiling, parameterName=Tile_AnimSpeed}))
  n23 = Cosine(None=@n22)
  NamedRerouteDeclaration(None=Saturate(None=LinearInterpolate(A=@n12, Alpha=@n23))) {name=RandomTiledSines, nodeColor=<rgba 0 1 0.8125 1>}
  n24 = Divide(A=Constant() {r=1}, B=ScalarParameter() {defaultValue=16, group=Tiling, parameterName=Tile})
  n25 = Fmod(A=@n18, B=@n24)
  n26 = Multiply(A=@n21.V2 Length, B=ScalarParameter() {defaultValue=68.759, group=Tiling, parameterName=RandomNumberX})
  n27 = Multiply(A=@n21.V2 Length, B=ScalarParameter() {defaultValue=125.15, group=Tiling, parameterName=RandomNumberY})
  n28 = AppendVector(
   A=LinearInterpolate(A=ScalarParameter() {defaultValue=0.125, group=Tiling, parameterName=LiftOffsetMin_X}, Alpha=Cosine(None=@n26))
   B=LinearInterpolate(A=ScalarParameter() {defaultValue=0.125, group=Tiling, parameterName=LiftOffsetMin_Y}, Alpha=Cosine(None=@n27))
  )
  n29 = ScalarParameter() {defaultValue=0.333, group=Tiling, parameterName=Amount_OffsetTile}
  NamedRerouteDeclaration(None=Add(A=Divide(A=@n25, B=@n24), B=Multiply(A=@n28, B=@n29))) {name=UVs, nodeColor=<rgba 0.1719 1 0 1>}
  n30 = ScalarParameter() {group=Glow, parameterName=GlowTIle_Scale_Min}
  n31 = ScalarParameter() {defaultValue=0.333, group=Glow, parameterName=GlowTile_Scale_Max}
  n32 = NamedRerouteUsage()
  n33 = TextureSample(UVs=MaterialFunctionCall(UVs=NamedRerouteUsage(), Texture Scale=LinearInterpolate(A=@n30, B=@n31, Alpha=@n32)) {materialFunction=&ScaleUVsByCenter}.UVs) {samplerType=SAMPLERTYPE_LinearColor, texture=&/T_UI_TacOvershield_Pattern_64x_copy}
  n34 = Multiply(A=NamedRerouteUsage(), B=ScalarParameter() {defaultValue=1, group=Glow, parameterName=GlowTile_Edge_Hardness})
  n35 = ScalarParameter() {defaultValue=-1, group=Glow, parameterName=GlowTile_FadePoint}
  n36 = ComponentMask(None=TextureCoordinate() {coordinateIndex=4}) {r=true}
  n37 = Multiply(A=Saturate(None=Multiply(A=@n33.R, B=@n34)), B=DotProduct(A=NamedRerouteUsage(), B=LinearInterpolate(A=@n35, Alpha=@n36) {constA=-1}))
  NamedRerouteDeclaration(None=Multiply(A=Saturate(None=@n37), B=ScalarParameter() {defaultValue=0.5, group=Glow, parameterName=GlowTiles_Alpha})) {name=GlowTiles, nodeColor=<rgba 0 1 0.2969 1>}
  n38 = ScalarParameter() {group=Solid, parameterName=SolidTile_Scale_Min}
  n39 = ScalarParameter() {defaultValue=0.333, group=Solid, parameterName=SolidTile_Scale_Max}
  n40 = NamedRerouteUsage()
  n41 = TextureSample(UVs=MaterialFunctionCall(UVs=NamedRerouteUsage(), Texture Scale=LinearInterpolate(A=@n38, B=@n39, Alpha=@n40)) {materialFunction=&ScaleUVsByCenter}.UVs) {samplerType=SAMPLERTYPE_LinearColor, texture=&T_UI_TacOvershield_Pattern_64x_copy}
  n42 = Multiply(A=NamedRerouteUsage(), B=ScalarParameter() {defaultValue=1, group=Solid, parameterName=SolidTile_Edge_Hardness})
  n43 = ScalarParameter() {defaultValue=-1, group=Solid, parameterName=SolidTile_FadePoint}
  n44 = ComponentMask(None=TextureCoordinate() {coordinateIndex=4}) {r=true}
  n45 = Multiply(A=Saturate(None=Multiply(A=@n41.G, B=@n42)), B=DotProduct(A=NamedRerouteUsage(), B=LinearInterpolate(A=@n43, Alpha=@n44) {constA=-1}))
  NamedRerouteDeclaration(None=Multiply(A=Saturate(None=@n45), B=ScalarParameter() {defaultValue=0.5, group=Solid, parameterName=SolidTiles_Alpha})) {name=SolidTiles, nodeColor=<rgba 0.1719 0 1 1>}
  n46 = ScalarParameter() {group=Outline, parameterName=OutlineTile_Scale_Min}
  n47 = ScalarParameter() {defaultValue=0.333, group=Outline, parameterName=OutlineTile_Scale_Max}
  n48 = NamedRerouteUsage()
  n49 = TextureSample(UVs=MaterialFunctionCall(UVs=NamedRerouteUsage(), Texture Scale=LinearInterpolate(A=@n46, B=@n47, Alpha=@n48)) {materialFunction=&ScaleUVsByCenter}.UVs) {samplerType=SAMPLERTYPE_LinearColor, texture=&T_UI_TacOvershield_Pattern_64x_copy}
  n50 = Multiply(A=NamedRerouteUsage(), B=ScalarParameter() {defaultValue=1, group=Outline, parameterName=OutlineTile_Edge_Hardness})
  n51 = ScalarParameter() {defaultValue=-1, group=Outline, parameterName=OutlineTile_FadePoint}
  n52 = ComponentMask(None=TextureCoordinate() {coordinateIndex=4}) {r=true}
  n53 = Multiply(A=Saturate(None=Multiply(A=@n49.B, B=@n50)), B=DotProduct(A=NamedRerouteUsage(), B=LinearInterpolate(A=@n51, Alpha=@n52) {constA=-1}))
  NamedRerouteDeclaration(None=Multiply(A=Saturate(None=@n53), B=ScalarParameter() {defaultValue=1, group=Outline, parameterName=OutlineTiles_Alpha})) {name=OutlineTiles, nodeColor=<rgba 1 0 0.1406 1>}
  n54 = ScalarParameter() {parameterName=Swipe_Start}
  n55 = ScalarParameter() {parameterName=Swipe_Stop}
  n56 = ScalarParameter() {parameterName=Animate_Swipe}
  n57 = Add(
   A=MaterialFunctionCall(UVs=TextureCoordinate() {coordinateIndex=4}, "Rotation Angle (0-1)"=ScalarParameter() {defaultValue=-0.125, parameterName=Swipe_Rotation}) {materialFunction=&CustomRotator}.Rotated Values
   B=AppendVector(A=LinearInterpolate(A=@n54, B=@n55, Alpha=@n56), B=Constant())
  )
  n58 = LinearInterpolate(A=Constant() {r=-3}, Alpha=ComponentMask(None=Sine(None=@n57) {period=2}) {r=true})
  n59 = DotProduct(A=NamedRerouteUsage(), B=NamedRerouteUsage())
  n60 = Saturate(None=DotProduct(A=Saturate(None=@n59), B=NamedRerouteUsage()))
  n61 = MaterialFunctionCall(
   Size=ScalarParameter() {defaultValue=0.5, group=Backing, parameterName=Size}
   Curvature=ScalarParameter() {defaultValue=0.75, group=Backing, parameterName=CornerRadius}
  ) {materialFunction=&MF_RoundedSquare}
  n62 = Saturate(None=MaterialFunctionCall(
    Alpha=Saturate(None=@n61.Linear Gradient)
    Min=ScalarParameter() {defaultValue=0.05, group=Backing, parameterName=Min}
    Max=ScalarParameter() {defaultValue=1, group=Backing, parameterName=Max}
   ) {materialFunction=&SmoothStep}.Result)
  n63 = ScalarParameter() {defaultValue=1, group=Backing, parameterName=EdgeHardness}
  n64 = NamedRerouteDeclaration(None=Saturate(None=Multiply(A=@n62, B=@n63))) {name=Backing, nodeColor=<rgba 0.1484 1 0 1>}
  n65 = Add(A=DotProduct(A=NamedRerouteUsage(), B=@n64), B=Multiply(A=@n64, B=ScalarParameter() {defaultValue=1, group=Backing, parameterName=Alpha}))
  n66 = StaticSwitchParameter(True=Saturate(None=DotProduct(A=@n58, B=@n60)), False=StaticSwitchParameter(True=@n60, False=Saturate(None=@n65)) {parameterName=IsOutline}) {parameterName=IsSwipe}
  Saturate(None=Multiply(
    A=StaticSwitchParameter(True=@n7, False=@n66) {parameterName=IsPattern}
    B=ScalarParameter() {defaultValue=1, group=Overall, parameterName=Final_Alpha}
   ))
  Sine(None=@n22)
  n67 = ScalarParameter() {group=Backing, parameterName=GradientOffsetStart}
  n68 = ScalarParameter() {defaultValue=-1, group=Backing, parameterName=GradientOffsetStop}
  n69 = ScalarParameter() {group=Backing, parameterName=Animate_Gradient_Size}
  n70 = TextureCoordinate() {coordinateIndex=4}
  n71 = Saturate(None=LinearInterpolate(A=LinearInterpolate(A=@n67, B=@n68, Alpha=@n69), Alpha=ComponentMask(None=@n70) {r=true}) {constA=-1})
  n72 = LinearInterpolate(
   A=VectorParameter() {group=Backing, parameterName=GradientColor_2, defaultValue=<rgba 0 0.1667 1 1>}.RGB
   B=VectorParameter() {group=Backing, parameterName=GradientColor_1, defaultValue=<rgba 0 1 1 1>}.RGB
   Alpha=@n71
  )
  n73 = VectorParameter() {group=Backing, parameterName=Backing_Color, defaultValue=<rgba 1 0 0 1>}
  n74 = Saturate(None=Power(Base=@n71, Exp=ScalarParameter() {defaultValue=1.5, group=Backing, parameterName=Gradient_Taper}))
  n75 = NamedRerouteDeclaration(None=Multiply(
    A=NamedRerouteDeclaration(None=@n74) {name=01Fade, nodeColor=<rgba 1 0 0.8203 1>}
    B=ScalarParameter() {group=Backing, parameterName=Animate_Gradient}
   )) {name=Gradient, nodeColor=<rgba 0 1 0.3203 1>}
  StaticSwitchParameter(
   True=@n72
   False=StaticSwitchParameter(True=NamedRerouteUsage(), False=LinearInterpolate(A=@n73.RGB, B=@n72, Alpha=@n75)) {parameterName=IsPattern}
  ) {parameterName=IsOutline}
 ]
}
