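# Player info HUD widget showing a small tactical overshield icon that animates to a broken icon when the overshield is depleted.
# Overshield value changes play the break animation forward at zero and in reverse on recharge, gated by a bool.
# When enabled, it binds pawn, DBNO and overshield-visibility delegates and listens for game-phase events to refresh visibility.
# It shows for a living, non-DBNO pawn, or without a pawn during the Aircraft phase.
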WidgetBlueprint /TacOvershield/UI/TacOvershieldPlayerInfo
animations [ShieldBroken_Anim {
  binding_count=2
  length_seconds=0.85
  object=&self:ShieldBroken_Anim
  sequence {
   bindings [
    Icon_Broken {
     id=#1
     possessed_class=Image
     tracks [
      {
       type=MovieSceneFloatTrack
       display_name=Render Opacity
       sections [{type=MovieSceneFloatSection, end_frame=25, start_frame=0, channels [None {type=MovieSceneScriptingFloatChannel, keys [[4.5,0],[9,0.5]]}]}]
      }
      {
       type=MovieScene2DTransformTrack
       sections [{
         type=MovieScene2DTransformSection
         end_frame=25
         start_frame=0
         channels [
          Translation.X {type=MovieSceneScriptingFloatChannel, const [0,1]}
          Translation.Y {type=MovieSceneScriptingFloatChannel, const [0,1]}
          Angle {type=MovieSceneScriptingFloatChannel, const [0,1]}
          Scale.X {type=MovieSceneScriptingFloatChannel, keys [[4.5,0.2],[12.001,1.4],[15,1]]}
          Scale.Y {type=MovieSceneScriptingFloatChannel, keys [[4.5,0.2],[12.001,1.4],[15,1]]}
          Shear.X {type=MovieSceneScriptingFloatChannel, const [0,1]}
          Shear.Y {type=MovieSceneScriptingFloatChannel, const [0,1]}
         ]
        }]
      }
     ]
    }
    Image_Icon {
     id=#2
     possessed_class=Image
     tracks [{
       type=MovieSceneColorTrack
       display_name=Color and Opacity
       sections [{
         type=MovieSceneColorSection
         end_frame=15
         start_frame=0
         channels [
          Color.R {type=MovieSceneScriptingFloatChannel, keys [[0,1],[8,0]]}
          Color.G {type=MovieSceneScriptingFloatChannel, keys [[0,1],[8,0]]}
          Color.B {type=MovieSceneScriptingFloatChannel, keys [[0,1],[8,0]]}
          Color.A {type=MovieSceneScriptingFloatChannel, keys [[0,1],[15,0]]}
         ]
        }]
      }]
    }
   ]
   display_rate {numerator=30}
  }
 }]
graphs [
 BindEvents {nodes [
   BindEvents!() -> (then, InPawn:Fort Player Pawn Object Reference) {is=title K2Node_FunctionEntry}
   > n1 = Sequence()
   > n2 = Get!(execute:Exec) -> (then:Exec, else:Exec, CachedPlayerPawn:Fort Player Pawn Object Reference)
   n3 = Create Event()
   > n4 = Unbind Event from On DBNOState Changed(self:Fort Pawn Object Reference=@n2, Delegate=@n3) {is=title K2Node_RemoveDelegate}
   n5 = Create Event()
   n6 = Get CachedPlayerState() -> (CachedPlayerState:Fort Player State Athena Object Reference)
   Unbind Event from On Overshield Visibility Changed(execute=[@n2.else, @n4.then], self:Fort Player State Zone Object Reference=@n6, Delegate=@n5) {is=title K2Node_RemoveDelegate}
   n7 = Create Event()
   > Unbind Event from On Pawn Set(self:Player State Object Reference=@n6, Delegate=@n7) {is=title K2Node_RemoveDelegate}
   n8 = Get!(execute=@n1.then_1) -> (then, else, InPawn:Fort Player Pawn Object Reference)
   > n9 = Bind Event to On DBNOState Changed(self:Fort Pawn Object Reference=@n8, Delegate=@n3) {is=title K2Node_AddDelegate}
   n10 = Get CachedPlayerState()
   Bind Event to On Overshield Visibility Changed(execute=[@n9.then, @n8.else], self:Fort Player State Zone Object Reference=@n10, Delegate=@n5) {is=title K2Node_AddDelegate}
   > Bind Event to On Pawn Set(self:Player State Object Reference=@n10, Delegate=@n7) {is=title K2Node_AddDelegate}
   > Set CachedPlayerPawn(CachedPlayerPawn:Fort Player Pawn Object Reference=Get InPawn() -> (InPawn:Fort Player Pawn Object Reference)) -> (Output_Get:Fort Player Pawn Object Reference)
  ]}
 EventGraph {nodes [
   Event Construct()
   > n1 = StartListeningToStatefulEvent!(
    execute:Exec
    Target:Gameplay Event Router Component Object Reference=GetGlobalEventRouter()
    EventType:Script Struct Object Reference=/Script/FortniteGame.FortGamePhaseUpdatedEvent
    Context:Object Reference
    EventBubblingRule:EEventBubblingRule Enum
   ) -> (
    then:Exec
    AsyncTaskProxy:Async Action Start Listening To Stateful Event Object Reference
    OnEventReceived:Exec
    OnSavedEventStateExists:Exec
    OnEventStateCleared:Exec
    Payload:Fort Game Phase Updated Event Structure
   ) {is=title K2Node_StartListeningToStatefulEventAsyncAction}
   > Set GamePhaseUpdatedAsyncTask(GamePhaseUpdatedAsyncTask:Async Action Start Listening To Stateful Event Object Reference=@n1) -> (Output_Get:Async Action Start Listening To Stateful Event Object Reference)
   Set Game Phase(
    execute=[@n1.OnEventReceived, @n1.OnSavedEventStateExists]
    Game Phase:EAthenaGamePhase Enum=Break Fort Game Phase Updated Event!(FortGamePhaseUpdatedEvent:"Fort Game Phase Updated Event Structure (by ref)"=@n1.Payload) -> (GamePhase:EAthenaGamePhase Enum) {is=title K2Node_BreakStruct}
   ) -> (Output_Get:EAthenaGamePhase Enum=None)
   > UpdateOvershieldVisibility()
   n2 = Event OnOvershieldStateChanged!() -> (OutputDelegate, then, Overshield:float, OvershieldMax:float) {is=title K2Node_Event}
   > UpdateOvershieldValue!(execute, self, Overshield:double=@n2.Overshield, MaxOvershield:double=@n2.OvershieldMax) -> (then)
   n3 = Event OnPlayerStateSet!() -> (OutputDelegate, then, InPlayerState:Fort Player State Athena Object Reference) {is=title K2Node_Event}
   > n4 = Is Valid(InputObject=@n3.InPlayerState)
   > Set CachedPlayerState(CachedPlayerState:Fort Player State Athena Object Reference=@n3.InPlayerState) -> (Output_Get:Fort Player State Athena Object Reference)
   > n5 = IsTacticalOvershieldEnabled!(execute) -> (then, ReturnValue:bool)
   > n6 = Branch(Condition=@n5)
   > n7 = GetPawnFromPlayerStateZone!(execute, InActor:Actor Object Reference=Get CachedPlayerState()) -> (then, OutPawn:Fort Player Pawn Object Reference, PawnReturned:bool)
   > BindEvents!(execute, self, InPawn:Fort Player Pawn Object Reference=@n7) -> (then) {is=title K2Node_CallFunction}
   > UpdateOvershieldVisibility()
   SetVisibility(execute=[@n6.else, @n4.Is Not Valid], InVisibility=Collapsed)
   Event Destruct()
   > n8 = Get!(execute) -> (then, else, GamePhaseUpdatedAsyncTask:Async Action Start Listening To Stateful Event Object Reference)
   > StopListeningToStatefulEvent(self:Async Action Start Listening To Stateful Event Object Reference=@n8)
  ]}
 GetLocalOvershieldVisibility {nodes [
   GetLocalOvershieldVisibility() {is=title K2Node_FunctionEntry}
   n1 = Get FortHUDContext!() -> (ReturnValue:Fort HUDContext Object Reference) {is=title K2Node_GetBlueprintContextEx}
   > n2 = Is Valid(InputObject=@n1)
   n3 = GetPlayerStateZone!(self:Fort HUDContext Object Reference=@n1) -> (ReturnValue:Fort Player State Zone Object Reference)
   > n4 = Is Valid(InputObject=@n3)
   Return Node!(execute=[@n4.Is Not Valid, @n2.Is Not Valid], ReturnValue:bool=false)
   > Return Node(ReturnValue=ShouldOvershieldBarBeVisible!(self:Fort Player State Zone Object Reference=@n3) -> (ReturnValue:bool))
  ]}
 OnDBNOStateChanged {nodes [
   OnDBNOStateChanged!() -> (then, FortPawn:Fort Pawn Object Reference, bInIsDBNO:bool=~) {is=title K2Node_FunctionEntry}
   > UpdateOvershieldVisibility()
  ]}
 OnPawnSet {nodes [
   n1 = OnPawnSet!() -> (then, Player:Player State Object Reference, NewPawn:Pawn Object Reference, OldPawn:Pawn Object Reference) {is=title K2Node_FunctionEntry}
   > n2 = Cast To FortPlayerPawn(Object=@n1.NewPawn)
   > BindEvents(InPawn=@n2)
   > UpdateOvershieldVisibility()
   UpdateOvershieldVisibility(execute=@n2.CastFailed)
  ]}
 OvershieldVisibilityChanged {nodes [
   OvershieldVisibilityChanged!() -> (then, bNewVisible:bool=~) {is=title K2Node_FunctionEntry}
   > UpdateOvershieldVisibility()
  ]}
 ResetShield {nodes [
   ResetShield() {is=title K2Node_FunctionEntry}
   > Set bShowingBroken(bShowingBroken:bool=false) -> (Output_Get:bool)
   > PlayAnimationReverse(InAnimation=Get ShieldBroken_Anim() -> (ShieldBroken_Anim:Widget Animation Object Reference))
  ]}
 TacOvershieldPlayerInfo_AutoGenFunc {nodes [TacOvershieldPlayerInfo_AutoGenFunc!() -> (then, FortPawn:Fort Pawn Object Reference, bInIsDBNO:bool=~) {is=title K2Node_FunctionEntry}]}
 UpdateOvershieldValue {nodes [
   n1 = UpdateOvershieldValue!() -> (then, Overshield:double=~, MaxOvershield:double=~) {is=title K2Node_FunctionEntry}
   > n2 = Branch(Condition="float > float"(A=@n1))
   > Branch(Condition=Get bShowingBroken() -> (bShowingBroken:bool))
   > Set bShowingBroken(bShowingBroken=false)
   n3 = Get ShieldBroken_Anim()
   > n4 = PlayAnimationReverse(InAnimation=@n3)
   InvalidateLayoutAndVolatility(execute=[@n4.then, @n6.then], self:Widget Object Reference)
   n5 = Branch(execute=@n2.else, Condition=Get bShowingBroken())
   Set bShowingBroken(execute=@n5.else, bShowingBroken=true)
   > n6 = PlayAnimationForward(InAnimation=@n3)
  ]}
 UpdateOvershieldVisibility {nodes [
   UpdateOvershieldVisibility() {is=title K2Node_FunctionEntry}
   > n1 = Branch()
   > n2 = Get!(execute) -> (then, else, CachedPlayerPawn:Fort Player Pawn Object Reference)
   n3 = IsDead!(self:Fort Pawn Object Reference=@n2) -> (ReturnValue:bool)
   n4 = IsDBNO!(self:Fort Pawn Object Reference=@n2) -> (ReturnValue:bool)
   > n5 = Branch(Condition=NOT Boolean(A=OR Boolean(A=@n4, B=@n3)))
   SetVisibility(execute=[@n5.then, @n6.then], InVisibility=HitTestInvisible)
   SetVisibility(execute=[@n5.else, @n6.else, @n1.else], InVisibility=Collapsed)
   n6 = Branch(
    execute=@n2.else
    Condition="Equal (Enum)"(A:EAthenaGamePhase Enum=Get Game Phase() -> (Game Phase:EAthenaGamePhase Enum=None), B:EAthenaGamePhase Enum=Aircraft)
   )
  ]}
]
variables [
 bCachedOvershieldVisible {type=boolean}
 bShouldBeVisible {type=boolean}
 bShowingBroken {type=boolean}
 CachedPlayerPawn {type=FortPlayerPawn}
 CachedPlayerState {type=FortPlayerStateAthena, default=null}
 Game Phase {type=EAthenaGamePhase, default=None}
 GamePhaseUpdatedAsyncTask {type=AsyncAction_StartListeningToStatefulEvent, default=null}
]
widgets {
 info {
  inheritedWidgetCount=1
  parentClass=/Script/TacOvershieldUI.TacOvershieldPlayerInfoWidget
  rootWidgetClass=&SizeBox
  widgetCount=5
 }
 widgets [
  SizeBox_0 {
   widgetClassPath=&SizeBox
   widget=SizeBox_0
   properties {bOverride_HeightOverride=true, bOverride_WidthOverride=true, heightOverride=20, widthOverride=20}
  }
  Overlay_0 {widgetClassPath=&Overlay, parent=SizeBox_0, widget=Overlay_0, slot {class=SizeBoxSlot}}
  Image_Icon {
   widgetClassPath=&Image
   bIsVariable=true
   parent=Overlay_0
   widget=Image_Icon
   properties {
    visibility=SelfHitTestInvisible
    brush {
     bIsDynamicallyLoaded=false
     drawAs=Image
     imageType=NoImage
     mirroring=NoMirror
     resourceName=None
     resourceObject=/TacOvershield/UI/Materials/M_UI_TacticalOvershield_Icon
     tiling=NoTile
     imageSize=<32 32>
     margin=<ltrb 0 0 0 0>
     outlineSettings {
      bUseBrushTransparency=false
      roundingType=HalfHeightRadius
      width=0
      color {colorUseRule=UseColor_Specified, specifiedColor=<rgba 0 0 0 0>}
      cornerRadii=<0 0 0 0>
     }
     tintColor {colorUseRule=UseColor_Specified, specifiedColor=<rgba 1 1 1 1>}
     uVRegion {bIsValid=false, max=<0 0>, min=<0 0>}
    }
   }
   slot {class=OverlaySlot}
  }
  Icon_Broken {
   widgetClassPath=&Image
   bIsVariable=true
   parent=Overlay_0
   widget=Icon_Broken
   properties {
    renderOpacity=0
    visibility=HitTestInvisible
    brush {
     bIsDynamicallyLoaded=false
     drawAs=Image
     imageType=NoImage
     mirroring=NoMirror
     resourceName=None
     resourceObject=/TacOvershield/UI/Materials/M_UI_TacticalOvershield_Broken_Icon
     tiling=NoTile
     imageSize=<45 45>
     margin=<ltrb 0 0 0 0>
     outlineSettings {
      bUseBrushTransparency=false
      roundingType=HalfHeightRadius
      width=0
      color {colorUseRule=UseColor_Specified, specifiedColor=<rgba 0 0 0 0>}
      cornerRadii=<0 0 0 0>
     }
     tintColor {colorUseRule=UseColor_Specified, specifiedColor=<rgba 0.0999 0.8469 1 1>}
     uVRegion {bIsValid=false, max=<0 0>, min=<0 0>}
    }
   }
   slot {class=OverlaySlot, properties {horizontalAlignment=HAlign_Center, verticalAlignment=VAlign_Center}}
  }
  TutorialHighlight {widgetClassPath=&TutorialHighlight}
 ]
}
