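# Spectator HUD widget, based on the Athena spectator hit point bar, that shows a player's tactical overshield through an embedded TacOvershieldPlayerInfo widget.
# Current-value changes pass straight into that widget's overshield update with the maximum fixed at one, so the value is treated as a fraction.
# DBNO changes collapse the widget while the player is downed and restore it as hit-test invisible otherwise.
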WidgetBlueprint /TacOvershield/UI/SpectatorTacOvershieldPlayerInfo
graphs [EventGraph {nodes [
   n1 = Event OnCurrentValueChanged!() -> (OutputDelegate, then, InCurrentValue:float) {is=title K2Node_Event}
   > UpdateOvershieldValue!(
    execute
    self:Tac Overshield Player Info Object Reference=Get TacOvershieldPlayerInfo() -> (TacOvershieldPlayerInfo:Tac Overshield Player Info Object Reference)
    Overshield:double=@n1.InCurrentValue
    MaxOvershield:double=1
   ) -> (then)
   n2 = Event OnDBNOChanged!() -> (OutputDelegate, then, bInIsDBNO:bool) {is=title K2Node_Event}
   > n3 = Branch(Condition=@n2.bInIsDBNO)
   n4 = Get TacOvershieldPlayerInfo()
   > SetVisibility(self=@n4, InVisibility=Collapsed)
   SetVisibility(execute=@n3.else, self=@n4, InVisibility=HitTestInvisible)
   Event PreConstruct()
   Event Construct()
   Event Tick()
  ]}]
variables []
widgets {
 info {
  inheritedWidgetCount=1
  parentClass=/Script/FortniteUI.AthenaSpectatorHitPointBarBase
  rootWidgetClass=&/TacOvershieldPlayerInfo.TacOvershieldPlayerInfo_C
  widgetCount=2
 }
 widgets [
  TacOvershieldPlayerInfo {
   widgetClassPath=&/TacOvershieldPlayerInfo
   bIsVariable=true
   widget=TacOvershieldPlayerInfo
   properties {elementTag {tagName=HUD.Athena.Broadcast.TacOvershieldPlayerInfo}}
  }
  ProgressBarDelta {widgetClassPath=/Script/FortniteUI.FortSimpleMaterialProgressBar, bInherited=true, widget=null}
 ]
}
