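# HUD widget for the tactical overshield, built on AthenaPlayerHitPointBarBase, showing the shield value with an animated icon.
# Value and delta changes drive Progress_Amount on the icon materials, while broken, auto-regen and full states trigger animations; the break animation plays in reverse to restore.
# Visibility follows the view target's overshield state and the Creative quickbar.
# Container width grows with the digit counts of max health, shield and overshield.
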WidgetBlueprint /TacOvershield/UI/TacOverShieldWidget
animations [
 A_ShieldCharge_BackingFXOnly {
  binding_count=3
  length_seconds=1
  object=&self:A_ShieldCharge_BackingFXOnly
  sequence {
   bindings [
    ShieldContainer {
     id=#1
     possessed_class=Image
     tracks [{
       type=MovieSceneWidgetMaterialTrack
       sections [{
         type=MovieSceneParameterSection
         channels [
          Animate_Gradient {type=MovieSceneScriptingFloatChannel, keys [[0,0],[12,1],[60,0]]}
          Animate_Gradient_Size {type=MovieSceneScriptingFloatChannel, keys [[0,1],[30,0],[60,0]]}
          GradientOffsetStart {type=MovieSceneScriptingFloatChannel, keys [[0,0],[30,0.375],[60,0]]}
          Alpha {type=MovieSceneScriptingFloatChannel, keys [[0,0.125],[30,0.1875],[60,0.125]]}
         ]
        }]
      }]
    }
    ContainerFX {
     id=#2
     possessed_class=Overlay
     tracks [{
       type=MovieSceneFloatTrack
       display_name=Render Opacity
       sections [{type=MovieSceneFloatSection, channels [None {type=MovieSceneScriptingFloatChannel, keys [[0,0],[15,1],[60,0]]}]}]
      }]
    }
    FX_BackingGlow_Swipe {
     id=#3
     possessed_class=Image
     tracks [
      {
       type=MovieSceneWidgetMaterialTrack
       sections [{type=MovieSceneParameterSection, channels [Animate_Swipe {type=MovieSceneScriptingFloatChannel, keys [[6,0],[60,1]]}]}]
      }
      {
       type=MovieSceneFloatTrack
       display_name=Render Opacity
       sections [{type=MovieSceneFloatSection, channels [None {type=MovieSceneScriptingFloatChannel, keys [[6,0],[15,1],[45,1],[60,0]]}]}]
      }
     ]
    }
   ]
   display_rate {numerator=60}
  }
 }
 A_ShieldFull_2 {
  binding_count=5
  length_seconds=1.75
  object=&self:A_ShieldFull_2
  sequence {bindings [
    Image_ShieldFull {
     id=#4
     possessed_class=Image
     tracks [
      {
       type=MovieSceneFloatTrack
       display_name=Render Opacity
       sections [{type=MovieSceneFloatSection, channels [None {type=MovieSceneScriptingFloatChannel, keys [[0,0],[5,1],[25,0]]}]}]
      }
      {
       type=MovieSceneWidgetMaterialTrack
       sections [{type=MovieSceneParameterSection, channels [AnimateProgressBackingColor {type=MovieSceneScriptingFloatChannel, keys [[0,0],[5,0],[8,1]]}]}]
      }
     ]
    }
    Icon_Broken_1 {
     id=#5
     possessed_class=Image
     tracks [{
       type=MovieSceneFloatTrack
       display_name=Render Opacity
       sections [{type=MovieSceneFloatSection, end_frame=35, start_frame=0, channels [None {type=MovieSceneScriptingFloatChannel, const [0,2]}]}]
      }]
    }
    ShieldFullFX {
     id=#6
     possessed_class=Overlay
     tracks [{
       type=MovieSceneFloatTrack
       display_name=Render Opacity
       sections [{type=MovieSceneFloatSection, channels [None {type=MovieSceneScriptingFloatChannel, keys [[0,0],[2.2973,1],[20.2973,1],[35,0]]}]}]
      }]
    }
    Icon_ShieldAnim {
     id=#7
     display_name=Icon_ShieldAnim_1
     possessed_class=Image
     tracks [
      {
       type=MovieScene2DTransformTrack
       sections [{
         type=MovieScene2DTransformSection
         channels [
          Translation.X {type=MovieSceneScriptingFloatChannel, const [0,2]}
          Translation.Y {type=MovieSceneScriptingFloatChannel, const [0,2]}
          Angle {type=MovieSceneScriptingFloatChannel, const [0,2]}
          Scale.X {type=MovieSceneScriptingFloatChannel, keys [[0,1],[19,1.6]]}
          Scale.Y {type=MovieSceneScriptingFloatChannel, keys [[0,1],[19,1.6]]}
          Shear.X {type=MovieSceneScriptingFloatChannel, const [0,2]}
          Shear.Y {type=MovieSceneScriptingFloatChannel, const [0,2]}
         ]
        }]
      }
      {
       type=MovieSceneFloatTrack
       display_name=Render Opacity
       sections [{type=MovieSceneFloatSection, channels [None {type=MovieSceneScriptingFloatChannel, keys [[0,0],[1,1],[19,0]]}]}]
      }
     ]
    }
   ]}
 }
 A_ShieldCharge_2 {
  binding_count=9
  length_seconds=1
  object=&self:A_ShieldCharge_2
  sequence {
   bindings [
    Icon_Shield {
     id=#8
     display_name=Icon_Progress
     possessed_class=Image
     tracks [{
       type=MovieSceneWidgetMaterialTrack
       sections [{
         type=MovieSceneParameterSection
         channels [
          AnimateBoltColor {type=MovieSceneScriptingFloatChannel, keys [[0,0],[9,1],[30,1],[60,0]]}
          AnimateBoltAlpha {type=MovieSceneScriptingFloatChannel, keys [[0,0],[9,1],[30,1],[60,0]]}
         ]
        }]
      }]
    }
    AnimBoundColorChange {
     id=#9
     possessed_class=Border
     tracks [{
       type=MovieSceneColorTrack
       display_name=Content Color and Opacity
       sections [{
         type=MovieSceneColorSection
         channels [
          Color.R {type=MovieSceneScriptingFloatChannel, keys [[0,1],[30,0],[60,1]]}
          Color.G {type=MovieSceneScriptingFloatChannel}
          Color.B {type=MovieSceneScriptingFloatChannel}
          Color.A {type=MovieSceneScriptingFloatChannel}
         ]
        }]
      }]
    }
    Icon_Broken_1 {
     id=#10
     possessed_class=Image
     tracks [{
       type=MovieSceneFloatTrack
       display_name=Render Opacity
       sections [{type=MovieSceneFloatSection, end_frame=60, start_frame=0, channels [None {type=MovieSceneScriptingFloatChannel, const [0,2]}]}]
      }]
    }
    ShieldContainer {
     id=#11
     possessed_class=Image
     tracks [{
       type=MovieSceneWidgetMaterialTrack
       sections [{
         type=MovieSceneParameterSection
         channels [
          Animate_Gradient {type=MovieSceneScriptingFloatChannel, keys [[0,0],[12,1],[60,0]]}
          Animate_Gradient_Size {type=MovieSceneScriptingFloatChannel, keys [[0,1],[30,0],[60,0]]}
          GradientOffsetStart {type=MovieSceneScriptingFloatChannel, keys [[0,0],[30,0.375],[60,0]]}
          Alpha {type=MovieSceneScriptingFloatChannel, keys [[0,0.125],[30,0.1875],[60,0.125]]}
         ]
        }]
      }]
    }
    FX_BackingGlow_Swipe {
     id=#12
     possessed_class=Image
     tracks [
      {
       type=MovieSceneWidgetMaterialTrack
       sections [{type=MovieSceneParameterSection, channels [Animate_Swipe {type=MovieSceneScriptingFloatChannel, keys [[6,0],[60,1]]}]}]
      }
      {
       type=MovieSceneFloatTrack
       display_name=Render Opacity
       sections [{type=MovieSceneFloatSection, channels [None {type=MovieSceneScriptingFloatChannel, keys [[6,0],[15,1],[45,1],[60,0]]}]}]
      }
     ]
    }
    Icon_Delta {
     id=#13
     possessed_class=Image
     tracks [{
       type=MovieSceneWidgetMaterialTrack
       sections [{
         type=MovieSceneParameterSection
         channels [
          AnimateBoltColor {type=MovieSceneScriptingFloatChannel, keys [[0,0],[3,1],[40,1],[60,0]]}
          BoltAlpha {type=MovieSceneScriptingFloatChannel, keys [[0,0],[9,1],[19,1],[60,0]]}
          AnimateProgressBackingColor {type=MovieSceneScriptingFloatChannel, keys [[0,0],[25,1],[60,0]]}
          AnimateProgressBackingAlpha {type=MovieSceneScriptingFloatChannel, keys [[0,0],[25,1],[60,0]]}
         ]
        }]
      }]
    }
    RoundGlow {
     id=#14
     display_name=RoundGlow_1
     possessed_class=Image
     tracks [{
       type=MovieSceneFloatTrack
       display_name=Render Opacity
       sections [{type=MovieSceneFloatSection, channels [None {type=MovieSceneScriptingFloatChannel, keys [[0,0],[9,1],[60,0]]}]}]
      }]
    }
    Icon_ShieldAnim {
     id=#7
     display_name=Icon_ShieldAnim_1
     possessed_class=Image
     tracks [
      {
       type=MovieScene2DTransformTrack
       sections [{
         type=MovieScene2DTransformSection
         channels [
          Scale.X {type=MovieSceneScriptingFloatChannel, keys [[6,1],[21,1.325]]}
          Scale.Y {type=MovieSceneScriptingFloatChannel, keys [[6,1],[21,1.325]]}
         ]
        }]
      }
      {
       type=MovieSceneFloatTrack
       display_name=Render Opacity
       sections [{type=MovieSceneFloatSection, channels [None {type=MovieSceneScriptingFloatChannel, keys [[6,0],[18,1],[60,0]]}]}]
      }
     ]
    }
    ContainerFX {
     id=#15
     possessed_class=Overlay
     tracks [{
       type=MovieSceneFloatTrack
       display_name=Render Opacity
       sections [{type=MovieSceneFloatSection, channels [None {type=MovieSceneScriptingFloatChannel, keys [[0,0],[15,1],[60,0]]}]}]
      }]
    }
   ]
   display_rate {numerator=60}
  }
 }
 A_DamageTaken {
  binding_count=1
  length_seconds=0.15
  object=&self:A_DamageTaken
  sequence {bindings [AnimBoundIconBorder {
     id=#16
     possessed_class=Border
     tracks [{
       type=MovieScene2DTransformTrack
       sections [{
         type=MovieScene2DTransformSection
         channels [
          Scale.X {type=MovieSceneScriptingFloatChannel, keys [[0,1],[1,1.125],[3,1]]}
          Scale.Y {type=MovieSceneScriptingFloatChannel, keys [[0,1],[1,1.125],[3,1]]}
         ]
        }]
      }]
    }]}
 }
 A_ShieldAssemble {
  binding_count=3
  length_seconds=0.25
  object=&self:A_ShieldAssemble
  sequence {bindings [
    Icon_Delta {
     id=#17
     possessed_class=Image
     tracks [{
       type=MovieSceneWidgetMaterialTrack
       sections [{
         type=MovieSceneParameterSection
         channels [
          Final_Alpha {type=MovieSceneScriptingFloatChannel, keys [[0,1],[0.3333,0]]}
          AnimateShieldColor {type=MovieSceneScriptingFloatChannel, keys [[0,0],[0.6667,11]]}
         ]
        }]
      }]
    }
    Icon_Progress {
     id=#18
     possessed_class=Image
     tracks [{
       type=MovieSceneWidgetMaterialTrack
       sections [{
         type=MovieSceneParameterSection
         channels [
          Final_Alpha {type=MovieSceneScriptingFloatChannel, keys [[0,1],[0.3333,0]]}
          AnimateShieldColor {type=MovieSceneScriptingFloatChannel, keys [[0,0],[0.6667,11]]}
         ]
        }]
      }]
    }
    Icon_Broken_1 {
     id=#19
     possessed_class=Image
     tracks [
      {
       type=MovieSceneWidgetMaterialTrack
       sections [{
         type=MovieSceneParameterSection
         channels [
          AnimateBrokenAlpha {type=MovieSceneScriptingFloatChannel, const [1,2]}
          AnimateBrokenColor {type=MovieSceneScriptingFloatChannel, const [0,2]}
         ]
        }]
      }
      {
       type=MovieSceneFloatTrack
       display_name=Render Opacity
       sections [{type=MovieSceneFloatSection, channels [None {type=MovieSceneScriptingFloatChannel, keys [[0,0],[0.3333,1]]}]}]
      }
      {
       type=MovieScene2DTransformTrack
       sections [{
         type=MovieScene2DTransformSection
         channels [
          Translation.X {type=MovieSceneScriptingFloatChannel, keys [[0,2],[0.3333,0]]}
          Angle {type=MovieSceneScriptingFloatChannel, keys [[0,-16],[0.3333,0]]}
          Scale.X {type=MovieSceneScriptingFloatChannel, keys [[0,1],[1,1.25],[3,0.95],[5,1]]}
          Scale.Y {type=MovieSceneScriptingFloatChannel, keys [[0,1],[1,1.25],[3,0.95],[5,1]]}
          Shear.X {type=MovieSceneScriptingFloatChannel, const [0,1]}
          Shear.Y {type=MovieSceneScriptingFloatChannel, const [0,1]}
         ]
        }]
      }
     ]
    }
   ]}
 }
 A_ShieldBroken_2 {
  binding_count=9
  length_seconds=0.9666
  object=&self:A_ShieldBroken_2
  sequence {bindings [
    OverlayTris2 {
     id=#20
     possessed_class=Overlay
     tracks [
      {
       type=MovieSceneFloatTrack
       display_name=Render Opacity
       sections [{type=MovieSceneFloatSection, channels [None {type=MovieSceneScriptingFloatChannel, keys [[1.187,0],[2.8537,1],[17.1873,1],[17.854,0]]}]}]
      }
      {
       type=MovieScene2DTransformTrack
       sections [{
         type=MovieScene2DTransformSection
         channels [
          Translation.X {type=MovieSceneScriptingFloatChannel}
          Translation.Y {type=MovieSceneScriptingFloatChannel}
          Angle {type=MovieSceneScriptingFloatChannel}
          Scale.X {type=MovieSceneScriptingFloatChannel, keys [[1.187,0],[17.187,1]]}
          Scale.Y {type=MovieSceneScriptingFloatChannel, keys [[1.187,0],[17.187,1]]}
          Shear.X {type=MovieSceneScriptingFloatChannel}
          Shear.Y {type=MovieSceneScriptingFloatChannel}
         ]
        }]
      }
     ]
    }
    TriFX_3 {
     id=#21
     possessed_class=Image
     tracks [{
       type=MovieSceneWidgetMaterialTrack
       sections [{type=MovieSceneParameterSection, channels [TriScale {type=MovieSceneScriptingFloatChannel, keys [[1.187,0.65],[18.8523,0]]}]}]
      }]
    }
    Icon_Broken_Shards_1 {
     id=#22
     possessed_class=Image
     tracks [
      {
       type=MovieSceneWidgetMaterialTrack
       sections [{
         type=MovieSceneParameterSection
         channels [
          Animate_Alpha {type=MovieSceneScriptingFloatChannel, keys [[0.187,1],[0.5203,0],[1.8537,0],[10.5203,1]]}
          Animate_Color {type=MovieSceneScriptingFloatChannel, keys [[0.5203,0],[1.8537,1],[7.8537,1],[9.8537,0]]}
          Animate_Multiply {type=MovieSceneScriptingFloatChannel, keys [[0.5203,0],[10.5203,1]]}
          Animate_Scale {type=MovieSceneScriptingFloatChannel, keys [[0.5203,0],[10.5203,1]]}
          Animate_Subtract {type=MovieSceneScriptingFloatChannel, keys [[0.5203,0],[4.5203,0.75],[10.5203,1]]}
         ]
        }]
      }
      {
       type=MovieScene2DTransformTrack
       sections [{type=MovieScene2DTransformSection, channels [Angle {type=MovieSceneScriptingFloatChannel, keys [[0.187,-4],[0.5203,0]]}]}]
      }
     ]
    }
    NoiseGlow_1 {
     id=#23
     possessed_class=Image
     tracks [{
       type=MovieSceneFloatTrack
       display_name=Render Opacity
       sections [{type=MovieSceneFloatSection, channels [None {type=MovieSceneScriptingFloatChannel, keys [[-0.1463,0],[0.5203,1],[1.8537,1],[5.187,0]]}]}]
      }]
    }
    Icon_Broken_1 {
     id=#24
     possessed_class=Image
     tracks [
      {
       type=MovieSceneWidgetMaterialTrack
       sections [{
         type=MovieSceneParameterSection
         channels [
          AnimateBrokenAlpha {type=MovieSceneScriptingFloatChannel, thinned_from=4, keys [[0.187,0],[3.8537,0],[18.5203,1]]}
          AnimateBrokenColor {type=MovieSceneScriptingFloatChannel, keys [[0.187,0],[2.187,1],[18.5203,0]]}
         ]
        }]
      }
      {
       type=MovieSceneFloatTrack
       display_name=Render Opacity
       sections [{type=MovieSceneFloatSection, channels [None {type=MovieSceneScriptingFloatChannel, keys [[-0.4797,0],[-0.1463,1]]}]}]
      }
      {
       type=MovieScene2DTransformTrack
       sections [{
         type=MovieScene2DTransformSection
         channels [
          Translation.X {type=MovieSceneScriptingFloatChannel, keys [[0.187,2],[0.5203,0]]}
          Angle {type=MovieSceneScriptingFloatChannel, keys [[0.187,-16],[0.5203,0]]}
          Scale.X {type=MovieSceneScriptingFloatChannel, keys [[0.187,1],[1.187,1.25],[3.187,0.95],[5.187,1]]}
          Scale.Y {type=MovieSceneScriptingFloatChannel, keys [[0.187,1],[1.187,1.25],[3.187,0.95],[5.187,1]]}
          Shear.X {type=MovieSceneScriptingFloatChannel, const [0,1]}
          Shear.Y {type=MovieSceneScriptingFloatChannel, const [0,1]}
         ]
        }]
      }
     ]
    }
    Flares_1 {
     id=#25
     possessed_class=Overlay
     tracks [
      {
       type=MovieSceneFloatTrack
       display_name=Render Opacity
       sections [{type=MovieSceneFloatSection, channels [None {type=MovieSceneScriptingFloatChannel, keys [[1.8537,0],[2.5203,1],[6.5203,0]]}]}]
      }
      {
       type=MovieScene2DTransformTrack
       sections [{
         type=MovieScene2DTransformSection
         channels [
          Scale.X {type=MovieSceneScriptingFloatChannel, keys [[1.8537,1],[6.5203,1.5]]}
          Scale.Y {type=MovieSceneScriptingFloatChannel, keys [[1.8537,1],[6.5203,1.5]]}
         ]
        }]
      }
     ]
    }
    TriFX {
     id=#26
     possessed_class=Image
     tracks [{
       type=MovieSceneWidgetMaterialTrack
       sections [{type=MovieSceneParameterSection, channels [TriScale {type=MovieSceneScriptingFloatChannel, keys [[1.187,0.65],[18.8523,0]]}]}]
      }]
    }
    Icon_Delta {
     id=#27
     possessed_class=Image
     tracks [{
       type=MovieSceneWidgetMaterialTrack
       sections [{
         type=MovieSceneParameterSection
         channels [
          Final_Alpha {type=MovieSceneScriptingFloatChannel, keys [[0.187,1],[0.5203,0]]}
          AnimateShieldColor {type=MovieSceneScriptingFloatChannel, keys [[0.187,0],[0.8537,11]]}
         ]
        }]
      }]
    }
    Icon_Progress {
     id=#28
     possessed_class=Image
     tracks [{
       type=MovieSceneWidgetMaterialTrack
       sections [{
         type=MovieSceneParameterSection
         channels [
          Final_Alpha {type=MovieSceneScriptingFloatChannel, keys [[0.187,1],[0.5203,0]]}
          AnimateShieldColor {type=MovieSceneScriptingFloatChannel, keys [[0.187,0],[0.8537,11]]}
         ]
        }]
      }]
    }
   ]}
 }
]
graphs [
 CheckForWidthUpdate {nodes [
   CheckForWidthUpdate() {is=title K2Node_FunctionEntry}
   n1 = Get FortHUDContext!() -> (ReturnValue:Fort HUDContext Object Reference) {is=title K2Node_GetBlueprintContextEx}
   > n2 = Is Valid(InputObject=@n1)
   FortShippingLog!(
    execute=@n2.Is Not Valid
    WorldContextObject:Object Reference=Self-Reference()
    InString:String=Check for Width Update: FortHUD invalid
    bLogAsWarning:bool=false
   ) -> (then)
   n3 = GetPlayerStateZone!(self:Fort HUDContext Object Reference=@n1) -> (ReturnValue:Fort Player State Zone Object Reference)
   n4 = Is Valid(exec=@n2.Is Valid, InputObject=@n3)
   FortShippingLog(execute=@n4.Is Not Valid, WorldContextObject=Self-Reference(), InString=Check for Width Update: PlayerStateZone invalid, bLogAsWarning=false)
   n5 = GetCurrentPawn!(self:Fort Player State Object Reference=@n3) -> (ReturnValue:Fort Player Pawn Object Reference)
   Is Valid(exec=@n4.Is Valid, InputObject=@n5)
   > n6 = GetHealthAndShields!(execute, self:Fort Player State Zone Object Reference=@n3) -> (then, Health:float, HealthMax:float, Shield:float, ShieldMax:float)
   n7 = GetMaxOvershield!(self:Fort Pawn Object Reference=@n5) -> (ReturnValue:float)
   n8 = Len(S="To String (Integer)"(InInt=Floor(A=@n7) {is=title K2Node_CallFunction}))
   n9 = float * float(A:int=int - int(A=@n8, B=Get DefaultOSCharCount() -> (DefaultOSCharCount:int=0)), B:int=15) {is=title K2Node_PromotableOperator}
   n10 = "integer <= integer"(A=@n8, B=Get DefaultOSCharCount() -> (DefaultOSCharCount=0))
   n11 = "Max (Float)"(A=@n6.HealthMax, B=float - float(A=@n6.ShieldMax, B=@n7))
   n12 = Len(S="To String (Integer)"(InInt=Floor(A=@n11)))
   n13 = "integer <= integer"(A=@n12, B=Get DefaultCharCount() -> (DefaultCharCount:int=0))
   n14 = Get DefaultCharCount() -> (DefaultCharCount=0)
   n15 = float + float(A=Get BaseWidgetWidth() -> (BaseWidgetWidth:double), B=float * float(A=int - int(A=@n12, B=@n14), B=15))
   n16 = Get BaseWidgetWidth()
   > SetWidthOverride(
    self=Get FullSizeBox() -> (FullSizeBox:Size Box Object Reference)
    InWidthOverride=float + float(A=SelectFloat(A=@n16, B=@n15, bPickA=@n13), B=SelectFloat(A=0, B=@n9, bPickA=@n10))
   )
  ]}
 EventGraph {nodes [
   EventOnFullState() {is=title K2Node_CustomEvent}
   > Set bBarIsFull(bBarIsFull:bool=true) -> (Output_Get:bool)
   > StopAnimation(InAnimation=Get A_ShieldBroken_2() -> (A_ShieldBroken_2:Widget Animation Object Reference))
   n1 = Get A_ShieldCharge_2() -> (A_ShieldCharge_2:Widget Animation Object Reference)
   > StopAnimation(InAnimation=@n1)
   > PlayAnimation(
    InAnimation=Get A_ShieldCharge_BackingFXOnly() -> (A_ShieldCharge_BackingFXOnly:Widget Animation Object Reference)
    StartAtTime=GetAnimationCurrentTime!(self:User Widget Object Reference, InAnimation:Widget Animation Object Reference=@n1) -> (ReturnValue:float)
   )
   > PlayAnimation(InAnimation=Get A_ShieldFull_2() -> (A_ShieldFull_2:Widget Animation Object Reference))
   n2 = Get A_ShieldBroken_2()
   > PlayAnimation(InAnimation=@n2, StartAtTime=GetEndTime(self=@n2), PlayMode=Reverse)
   n3 = OnOvershieldVisibilityChanged_Event_ViewTarget!() -> (OutputDelegate, then, bNewVisible:bool=~) {is=title K2Node_CustomEvent}
   > OnOvershieldVisibilityChanged!(execute, self, bNewVisible:bool=@n3.bNewVisible) -> (then)
   n4 = OnViewTargetChanged_Event_0!() -> (
    OutputDelegate
    then
    PlayerController:Fort Player Controller Object Reference
    OldViewTarget:Actor Object Reference
    NewViewTarget:Actor Object Reference
   ) {is=title K2Node_CustomEvent}
   > n5 = Cast To FortPlayerPawnAthena(Object=@n4.NewViewTarget)
   > n6 = Cast To FortPlayerStateZone!(execute, Object=Get PlayerState!(self:Pawn Object Reference=@n5) -> (PlayerState:Player State Object Reference)) -> (then, CastFailed, AsFort Player State Zone:Fort Player State Zone Object Reference) {is=title K2Node_DynamicCast}
   > Assign On Overshield Visibility Changed!(execute:Exec, self:Fort Player State Zone Object Reference=@n6, Delegate:"Delegate (by ref)"=@n3) -> (then:Exec) {is=title K2Node_AssignDelegate}
   > OnOvershieldVisibilityChanged(bNewVisible=ShouldOvershieldBarBeVisible!(self:Fort Player State Zone Object Reference=@n6) -> (ReturnValue:bool))
   FortShippingLog(execute=@n6.CastFailed, WorldContextObject=Self-Reference(), InString=FortPlayerStateZone cast failed, bLogAsWarning=false)
   FortShippingLog(execute=@n5.CastFailed, WorldContextObject=Self-Reference(), InString=FortPlayerPawn cast failed, bLogAsWarning=false)
   n7 = OnLocalPlayerMaxHealthChanged_Event!() -> (OutputDelegate, then, HitPointStat:float=~) {is=title K2Node_CustomEvent}
   CheckForWidthUpdate(execute=[@n8.then, @n7.then]) {is=title K2Node_CallFunction}
   n8 = OnLocalPlayerMaxShieldChanged_Event!() -> (OutputDelegate, then, HitPointStat:float=~) {is=title K2Node_CustomEvent}
   Event PreConstruct()
   > n9 = Branch(Condition=IsUsingTouch(PlayerController=GetOwningPlayer()))
   n10 = Get FullSizeBox()
   > Set BaseWidgetWidth(BaseWidgetWidth:double=float - float(A=Get WidthOverride!(self:Size Box Object Reference=@n10) -> (WidthOverride:float), B=35)) -> (Output_Get:double)
   Set BaseWidgetWidth(execute=@n9.else, BaseWidgetWidth=Get WidthOverride(self=Get FullSizeBox()))
   Event Construct()
   > n11 = Sequence()
   > n12 = IsTacticalOvershieldEnabled!(execute) -> (then, ReturnValue:bool)
   > Branch(Condition=@n12)
   > n13 = Cast To FortPlayerController(Object=GetOwningPlayer())
   > Assign On View Target Changed!(execute:Exec, self:Fort Player Controller Object Reference=@n13, Delegate:"Delegate (by ref)"=@n4) -> (then:Exec) {is=title K2Node_AssignDelegate}
   n14 = Get FortHUDContext()
   > OnOvershieldVisibilityChanged(bNewVisible=ShouldOvershieldBarBeVisible(self=GetPlayerStateZone(self=@n14)))
   n15 = GetViewTarget!(self:Controller Object Reference=@n13) -> (ReturnValue:Actor Object Reference)
   > Is Valid(InputObject=@n15)
   > OnViewTargetChanged_Event_0!(
    execute
    self
    PlayerController:Fort Player Controller Object Reference
    OldViewTarget:Actor Object Reference
    NewViewTarget:Actor Object Reference=@n15
   ) -> (then) {is=title K2Node_CallFunction}
   FortShippingLog(execute=@n13.CastFailed, WorldContextObject=Self-Reference(), InString=FortPlayerController cast failed, bLogAsWarning=false)
   Bind Event to On Creative Quickbar Equipped Changed(
    execute=@n11.then_1
    self:Creative HUDLContext Object Reference=Get CreativeHUDLContext!() -> (ReturnValue:Creative HUDLContext Object Reference) {is=title K2Node_GetBlueprintContextEx}
    Delegate=Create Event()
   ) {is=title K2Node_AddDelegate}
   > Bind Event to On Local Player Health Changed(self:Fort HUDContext Object Reference=Get FortHUDContext(), Delegate=Create Event()) {is=title K2Node_AddDelegate}
   n16 = Get FortHUDContext()
   > Assign On Local Player Max Health Changed!(execute:Exec, self:Fort HUDContext Object Reference=@n16, Delegate:"Delegate (by ref)"=@n7) -> (then:Exec) {is=title K2Node_AssignDelegate}
   > Bind Event to On Local Player Max Shield Changed(self:Fort HUDContext Object Reference=@n16, Delegate=@n8) {is=title K2Node_AddDelegate}
   > CheckForWidthUpdate()
   Event Destruct()
   n17 = Get CreativeHUDLContext()
   > Is Valid(InputObject=@n17)
   > Unbind Event from On Creative Quickbar Equipped Changed(self:Creative HUDLContext Object Reference=@n17, Delegate=Create Event()) {is=title K2Node_RemoveDelegate}
   n18 = Get CreativeHUDLContext()
   > Is Valid(InputObject=@n18)
   n19 = Get FortHUDContext()
   > Is Valid(InputObject=@n19)
   > Unbind Event from On Local Player Health Changed(self:Fort HUDContext Object Reference=@n19, Delegate=Create Event()) {is=title K2Node_RemoveDelegate}
   Event OnMaxValueChanged!() -> (OutputDelegate, then, Value:float) {is=title K2Node_Event}
   > UpdateFillBar()
   > UpdateDeltaBar()
   > UpdateValue()
   > CheckForWidthUpdate()
   n20 = Event OnValueChangedWithReason!() -> (OutputDelegate, then, Value:float, Reason:EFortHitPointModificationReason Enum=Invalid) {is=title K2Node_Event}
   > UpdateFillBar()
   > UpdateValue()
   n21 = Get bBarIsFull() -> (bBarIsFull:bool)
   > n22 = Branch(Condition=AND Boolean(
     A="Equal (Enum)"(A:EFortHitPointModificationReason Enum=@n20.Reason, B:EFortHitPointModificationReason Enum=AutoRegen)
     B=NOT Boolean(A=@n21)
    ))
   > SetRegenState()
   InterruptRegen(execute=@n22.else)
   Event OnDeltaChanged() {is=title K2Node_Event}
   > UpdateDeltaBar()
   Reroute Node(InputPin:Creative HUDLContext Object Reference=@n18) -> (OutputPin:Creative HUDLContext Object Reference)
  ]}
 InterruptRegen {nodes [
   InterruptRegen() {is=title K2Node_FunctionEntry}
   > n1 = Sequence()
   n2 = Get A_ShieldCharge_BackingFXOnly()
   Branch(execute=@n1.then_1, Condition=IsAnimationPlaying(InAnimation=@n2))
   > StopAnimation(InAnimation=@n2)
   n3 = Get A_ShieldCharge_2()
   Branch(execute=@n1.then_0, Condition=IsAnimationPlaying(InAnimation=@n3))
   > StopAnimation(InAnimation=@n3)
  ]}
 OnCreativeQuickbarEquippedChanged {nodes [
   n1 = OnCreativeQuickbarEquippedChanged!() -> (then, bIsQuickbarEquipped:bool=~) {is=title K2Node_FunctionEntry}
   > n2 = Set bCachedQuickbarEquipped(bCachedQuickbarEquipped:bool=@n1) -> (Output_Get:bool)
   > n3 = Branch(Condition=@n2)
   > SetVisibility(InVisibility=Hidden)
   n4 = IsTacticalOvershieldEnabled(execute=@n3.else)
   n5 = ShouldOvershieldBarBeVisible(self=GetPlayerStateZone(self=Get FortHUDContext()))
   > SetVisibility(InVisibility=Select(Option 0=Hidden, Option 1=HitTestInvisible, Index=AND Boolean(A=@n4, B=@n5)))
  ]}
 OnHealthChanged {nodes [
   n1 = OnHealthChanged!() -> (then, HitPointValue:float=~, ModificationReason:EFortHitPointModificationReason Enum=Invalid) {is=title K2Node_FunctionEntry}
   > Branch(Condition="Equal (Enum)"(A=@n1.ModificationReason, B=DamageReceived))
   > InterruptRegen() {is=title K2Node_CallFunction}
  ]}
 OnOvershieldVisibilityChanged {nodes [
   n1 = OnOvershieldVisibilityChanged!() -> (then, bNewVisible:bool=~) {is=title K2Node_FunctionEntry}
   n2 = Get bCachedQuickbarEquipped() -> (bCachedQuickbarEquipped:bool)
   > n3 = Branch(Condition=AND Boolean(A=@n1, B=NOT Boolean(A=@n2)))
   > SetVisibility(InVisibility=HitTestInvisible)
   SetVisibility(execute=@n3.else, InVisibility=Hidden)
  ]}
 SetRegenState {nodes [
   SetRegenState() {is=title K2Node_FunctionEntry}
   > n1 = Branch(Condition=IsAnimationPlaying(InAnimation=Get A_ShieldCharge_2()))
   PlayAnimation(execute=@n1.else, InAnimation=Get A_ShieldCharge_2(), NumLoopsToPlay=0)
  ]}
 UpdateDeltaBar {nodes [
   UpdateDeltaBar() {is=title K2Node_FunctionEntry}
   > n1 = GetDynamicMaterial(self=Get Icon_Delta() -> (Icon_Delta:Image Object Reference))
   > SetScalarParameterValue(
    self=@n1
    ParameterName=Progress_Amount
    Value=GetLastValuePercentage!(self:Athena Player Hit Point Bar Base Object Reference) -> (ReturnValue:float)
   )
  ]}
 UpdateFillBar {nodes [
   UpdateFillBar() {is=title K2Node_FunctionEntry}
   > n1 = GetDynamicMaterial(self=Get Icon_Progress() -> (Icon_Progress:Image Object Reference))
   n2 = GetCurrentValuePercentage!(self:Athena Player Hit Point Bar Base Object Reference) -> (ReturnValue:float)
   > SetScalarParameterValue(self=@n1, ParameterName=Progress_Amount, Value=@n2)
   > UpdateShieldBrokenStatus!(execute, self, bShieldBroken:bool="float <= float"(A=@n2) -> (ReturnValue=~)) -> (then)
   > n3 = Branch(Condition="Nearly Equal (Float)"(A=@n2, B=1, ErrorTolerance=0))
   > EventOnFullState() {is=title K2Node_CallFunction}
   Set bBarIsFull(execute=@n3.else, bBarIsFull=false)
  ]}
 UpdateShieldBrokenStatus {nodes [
   n1 = UpdateShieldBrokenStatus!() -> (then, bShieldBroken:bool=~) {is=title K2Node_FunctionEntry}
   > Branch(Condition="Not Equal (Boolean)"(A=@n1, B=Get bIsBroken() -> (bIsBroken:bool)) -> (ReturnValue=~))
   > Set bIsBroken(bIsBroken:bool=@n1) -> (Output_Get:bool)
   > n2 = Branch(Condition=Get bIsBroken())
   > PlayAnimation(InAnimation=Get A_ShieldBroken_2())
   n3 = Get A_ShieldBroken_2()
   n4 = Branch(execute=@n2.else, Condition=IsAnimationPlaying(InAnimation=@n3))
   > n5 = StopAnimation(InAnimation=@n3)
   n6 = Get A_ShieldBroken_2()
   PlayAnimation(execute=[@n5.then, @n4.else], InAnimation=@n6, StartAtTime=GetEndTime(self=@n6), PlayMode=Reverse)
  ]}
 UpdateValue {nodes [
   UpdateValue() {is=title K2Node_FunctionEntry}
   n1 = Get ValueCurrent() -> (ValueCurrent:float)
   n2 = Round(A=@n1) {is=title K2Node_CallFunction}
   > n3 = Branch(Condition=AND Boolean(A="Equal (Integer)"(A=@n2, B=0) {is=title K2Node_CallFunction}, B="float > float"(A=@n1, B=0) {is=title K2Node_CallFunction}))
   n4 = Get CurrentValue() -> (CurrentValue:Common Text Object Reference)
   > "SetText (Text)"(self=@n4, InText="To Text (Integer)"(Value=1))
   "SetText (Text)"(execute=@n3.else, self=@n4, InText="To Text (Integer)"(Value=@n2))
  ]}
]
variables [
 ActiveShieldIndex {type=integer, default=1}
 AssembleAnimHasPlayed {type=boolean}
 BaseWidgetWidth {type=number}
 bBarIsFull {type=boolean}
 bCachedQuickbarEquipped {type=boolean}
 bIsBroken {type=boolean}
 bRegen {type=boolean}
 BrokenStyle {
  type=ProgressBarStyle
  default {
   enableFillAnimation=false
   backgroundImage {
    bIsDynamicallyLoaded=false
    drawAs=Box
    imageType=FullColor
    mirroring=NoMirror
    resourceName=../../../Engine/Content/Slate/Common/ProgressBar_Background.png
    resourceObject=null
    tiling=NoTile
    imageSize=<0 0>
    margin=<ltrb 0.4167 0.4167 0.4167 0.4167>
    outlineSettings {
     bUseBrushTransparency=false
     roundingType=HalfHeightRadius
     width=0
     color {colorUseRule=UseColor_Specified, specifiedColor=<rgba 0 0 0 0>}
     cornerRadii=<0 0 0 1>
    }
    tintColor {colorUseRule=UseColor_Specified, specifiedColor=<rgba 1 0 0 0.5>}
    uVRegion {bIsValid=false, max=<0 0>, min=<0 0>}
   }
   fillImage {
    bIsDynamicallyLoaded=false
    drawAs=Box
    imageType=FullColor
    mirroring=NoMirror
    resourceName=../../../Engine/Content/Slate/Common/ProgressBar_Fill.png
    resourceObject=null
    tiling=NoTile
    imageSize=<0 0>
    margin=<ltrb 0.4167 0.4167 0.4167 0.4167>
    outlineSettings {
     bUseBrushTransparency=false
     roundingType=HalfHeightRadius
     width=0
     color {colorUseRule=UseColor_Specified, specifiedColor=<rgba 0 0 0 0>}
     cornerRadii=<0 0 0 1>
    }
    tintColor {colorUseRule=UseColor_Specified, specifiedColor=<rgba 1 1 1 0.85>}
    uVRegion {bIsValid=false, max=<0 0>, min=<0 0>}
   }
   marqueeImage {
    bIsDynamicallyLoaded=false
    drawAs=Image
    imageType=FullColor
    mirroring=NoMirror
    resourceName=../../../Engine/Content/Slate/Common/ProgressBar_Marquee.png
    resourceObject=null
    tiling=Horizontal
    imageSize=<20 12>
    margin=<ltrb 0 0 0 0>
    outlineSettings {
     bUseBrushTransparency=false
     roundingType=HalfHeightRadius
     width=0
     color {colorUseRule=UseColor_Specified, specifiedColor=<rgba 0 0 0 0>}
     cornerRadii=<0 0 0 1>
    }
    tintColor {colorUseRule=UseColor_Specified, specifiedColor=<rgba 1 1 1 1>}
    uVRegion {bIsValid=false, max=<0 0>, min=<0 0>}
   }
  }
 }
 bUseIconVersion {type=boolean, default=true}
 DefaultCharCount {type=integer, default=3}
 DefaultOSCharCount {type=integer, default=2}
 DefaultStyle {
  type=ProgressBarStyle
  default {
   enableFillAnimation=false
   backgroundImage {
    bIsDynamicallyLoaded=false
    drawAs=Box
    imageType=FullColor
    mirroring=NoMirror
    resourceName=../../../Engine/Content/Slate/Common/ProgressBar_Background.png
    resourceObject=null
    tiling=NoTile
    imageSize=<0 0>
    margin=<ltrb 0.4167 0.4167 0.4167 0.4167>
    outlineSettings {
     bUseBrushTransparency=false
     roundingType=HalfHeightRadius
     width=0
     color {colorUseRule=UseColor_Specified, specifiedColor=<rgba 0 0 0 0>}
     cornerRadii=<0 0 0 1>
    }
    tintColor {colorUseRule=UseColor_Specified, specifiedColor=<rgba 0 0 0 0.25>}
    uVRegion {bIsValid=false, max=<0 0>, min=<0 0>}
   }
   fillImage {
    bIsDynamicallyLoaded=false
    drawAs=Box
    imageType=FullColor
    mirroring=NoMirror
    resourceName=../../../Engine/Content/Slate/Common/ProgressBar_Fill.png
    resourceObject=null
    tiling=NoTile
    imageSize=<0 0>
    margin=<ltrb 0.4167 0.4167 0.4167 0.4167>
    outlineSettings {
     bUseBrushTransparency=false
     roundingType=HalfHeightRadius
     width=0
     color {colorUseRule=UseColor_Specified, specifiedColor=<rgba 0 0 0 0>}
     cornerRadii=<0 0 0 1>
    }
    tintColor {colorUseRule=UseColor_Specified, specifiedColor=<rgba 1 1 1 0.85>}
    uVRegion {bIsValid=false, max=<0 0>, min=<0 0>}
   }
   marqueeImage {
    bIsDynamicallyLoaded=false
    drawAs=Image
    imageType=FullColor
    mirroring=NoMirror
    resourceName=../../../Engine/Content/Slate/Common/ProgressBar_Marquee.png
    resourceObject=null
    tiling=Horizontal
    imageSize=<20 12>
    margin=<ltrb 0 0 0 0>
    outlineSettings {
     bUseBrushTransparency=false
     roundingType=HalfHeightRadius
     width=0
     color {colorUseRule=UseColor_Specified, specifiedColor=<rgba 0 0 0 0>}
     cornerRadii=<0 0 0 1>
    }
    tintColor {colorUseRule=UseColor_Specified, specifiedColor=<rgba 1 1 1 1>}
    uVRegion {bIsValid=false, max=<0 0>, min=<0 0>}
   }
  }
 }
 HideValueTimerHandle {type=TimerHandle}
]
widgets {
 info {
  inheritedWidgetCount=1
  namedSlotCount=1
  parentClass=/Script/FortniteUI.AthenaPlayerHitPointBarBase
  rootWidgetClass=/Game/Athena/HUD/FortHUDElementWrapper.FortHUDElementWrapper_C
  widgetCount=43
 }
 widgets [
  FortHUDElementWrapper {
   widgetClassPath=/Game/Athena/HUD/FortHUDElementWrapper
   bIsVariable=true
   widget=FortHUDElementWrapper
   properties {
    hiddenVisibility=Hidden
    elementTag {tagName=HUD.Athena.Player.Overshield}
    hUDElementTag {gameplayTags [
      {tagName=Athena.HideForCapture}
      {tagName=HUD.Athena.HideForPostGame}
      {tagName=HUD.Athena.HideforRadial}
      {tagName=HUD.Athena.HideOnBus}
      {tagName=HUD.Event.HideForSpecialEvent}
      {tagName=HUD.Athena.Hide.WithInventory}
      {tagName=HUD.Hide.WhileLoading}
      {tagName=HUD.Athena.Hide.WithMap}
      {tagName=Athena.Creative.HideForLoading}
      {tagName=HUD.TopLeft.PlayerHitPoints}
      {tagName=Athena.Creative.HUDController.HideShields}
      {tagName=HUD.Hide.CreativeJoiningLobby}
     ]}
   }
  }
  Overlay_1 {widgetClassPath=&Overlay, bIsVariable=true, namedSlotHost=FortHUDElementWrapper, widget=Overlay_1}
  FullSizeBox {
   widgetClassPath=&SizeBox
   bIsVariable=true
   parent=Overlay_1
   widget=FullSizeBox
   properties {bOverride_HeightOverride=true, bOverride_WidthOverride=true, heightOverride=80, maxDesiredWidth=570, widthOverride=540}
   slot {class=OverlaySlot, properties {horizontalAlignment=HAlign_Fill, verticalAlignment=VAlign_Fill}}
  }
  ShieldContainer {
   widgetClassPath=&Image
   bIsVariable=true
   parent=FullSizeBox
   widget=ShieldContainer
   properties {brush {
     bIsDynamicallyLoaded=false
     drawAs=Box
     imageType=NoImage
     mirroring=NoMirror
     resourceName=None
     resourceObject=/TacOvershield/UI/Materials/M_UI_TacOvershield_Backing
     tiling=NoTile
     imageSize=<80 80>
     margin=<ltrb 0.5 0.5 0.5 0.5>
     outlineSettings {
      bUseBrushTransparency=false
      roundingType=FixedRadius
      width=0
      color {colorUseRule=UseColor_Specified, specifiedColor=<rgba 0 0 0 0>}
      cornerRadii=<12 12 12 12>
     }
     tintColor {colorUseRule=UseColor_Specified, specifiedColor=<rgba 1 1 1 1>}
     uVRegion {bIsValid=false, max=<0 0>, min=<0 0>}
    }}
   slot {class=SizeBoxSlot}
  }
  ContainerFX {
   widgetClassPath=&Overlay
   parent=Overlay_1
   widget=ContainerFX
   properties {renderOpacity=0}
   slot {class=OverlaySlot, properties {horizontalAlignment=HAlign_Fill, verticalAlignment=VAlign_Fill}}
  }
  FX_GlowPattern {
   widgetClassPath=&Image
   bIsVariable=true
   parent=ContainerFX
   widget=FX_GlowPattern
   properties {brush {
     bIsDynamicallyLoaded=false
     drawAs=Box
     imageType=NoImage
     mirroring=NoMirror
     resourceName=None
     resourceObject=/TacOvershield/UI/Materials/M_UI_TacOvershield_Pattern_Glow_Large
     tiling=NoTile
     imageSize=<80 80>
     margin=<ltrb 0.5 0.5 0.5 0.5>
     outlineSettings {
      bUseBrushTransparency=false
      roundingType=FixedRadius
      width=0
      color {colorUseRule=UseColor_Specified, specifiedColor=<rgba 0 0 0 0>}
      cornerRadii=<12 12 12 12>
     }
     tintColor {colorUseRule=UseColor_Specified, specifiedColor=<rgba 1 1 1 1>}
     uVRegion {bIsValid=false, max=<0 0>, min=<0 0>}
    }}
   slot {class=OverlaySlot, properties {horizontalAlignment=HAlign_Fill, verticalAlignment=VAlign_Fill}}
  }
  FX_SolidPattern_Large {
   widgetClassPath=&Image
   bIsVariable=true
   parent=ContainerFX
   widget=FX_SolidPattern_Large
   properties {brush {
     bIsDynamicallyLoaded=false
     drawAs=Box
     imageType=NoImage
     mirroring=NoMirror
     resourceName=None
     resourceObject=/TacOvershield/UI/Materials/M_UI_TacOvershield_Pattern_Solid_Large
     tiling=NoTile
     imageSize=<80 80>
     margin=<ltrb 0.5 0.5 0.5 0.5>
     outlineSettings {
      bUseBrushTransparency=false
      roundingType=FixedRadius
      width=0
      color {colorUseRule=UseColor_Specified, specifiedColor=<rgba 0 0 0 0>}
      cornerRadii=<12 12 12 12>
     }
     tintColor {colorUseRule=UseColor_Specified, specifiedColor=<rgba 1 1 1 1>}
     uVRegion {bIsValid=false, max=<0 0>, min=<0 0>}
    }}
   slot {class=OverlaySlot, properties {horizontalAlignment=HAlign_Fill, verticalAlignment=VAlign_Fill}}
  }
  FX_SolidPattern_Small {
   widgetClassPath=&Image
   bIsVariable=true
   parent=ContainerFX
   widget=FX_SolidPattern_Small
   properties {brush {
     bIsDynamicallyLoaded=false
     drawAs=Box
     imageType=NoImage
     mirroring=NoMirror
     resourceName=None
     resourceObject=/TacOvershield/UI/Materials/M_UI_TacOvershield_Pattern_Solid_Small
     tiling=NoTile
     imageSize=<80 80>
     margin=<ltrb 0.5 0.5 0.5 0.5>
     outlineSettings {
      bUseBrushTransparency=false
      roundingType=FixedRadius
      width=0
      color {colorUseRule=UseColor_Specified, specifiedColor=<rgba 0 0 0 0>}
      cornerRadii=<12 12 12 12>
     }
     tintColor {colorUseRule=UseColor_Specified, specifiedColor=<rgba 1 1 1 1>}
     uVRegion {bIsValid=false, max=<0 0>, min=<0 0>}
    }}
   slot {class=OverlaySlot, properties {horizontalAlignment=HAlign_Fill, verticalAlignment=VAlign_Fill}}
  }
  FX_OutlinePattern_Large {
   widgetClassPath=&Image
   bIsVariable=true
   parent=ContainerFX
   widget=FX_OutlinePattern_Large
   properties {
    brush {
     bIsDynamicallyLoaded=false
     drawAs=Box
     imageType=NoImage
     mirroring=NoMirror
     resourceName=None
     resourceObject=/TacOvershield/UI/Materials/M_UI_TacOvershield_Pattern_Outline_Large
     tiling=NoTile
     imageSize=<80 80>
     margin=<ltrb 0.5 0.5 0.5 0.5>
     outlineSettings {
      bUseBrushTransparency=false
      roundingType=FixedRadius
      width=0
      color {colorUseRule=UseColor_Specified, specifiedColor=<rgba 0 0 0 0>}
      cornerRadii=<12 12 12 12>
     }
     tintColor {colorUseRule=UseColor_Specified, specifiedColor=<rgba 1 1 1 1>}
     uVRegion {bIsValid=false, max=<0 0>, min=<0 0>}
    }
    renderTransform {angle=0, scale=<1 1>, shear=<0 0>, translation=<0 1>}
   }
   slot {class=OverlaySlot, properties {horizontalAlignment=HAlign_Fill, verticalAlignment=VAlign_Fill}}
  }
  FX_OutlinePattern_Small {
   widgetClassPath=&Image
   bIsVariable=true
   parent=ContainerFX
   widget=FX_OutlinePattern_Small
   properties {
    brush {
     bIsDynamicallyLoaded=false
     drawAs=Box
     imageType=NoImage
     mirroring=NoMirror
     resourceName=None
     resourceObject=/TacOvershield/UI/Materials/M_UI_TacOvershield_Pattern_Outline_Small
     tiling=NoTile
     imageSize=<80 80>
     margin=<ltrb 0.5 0.5 0.5 0.5>
     outlineSettings {
      bUseBrushTransparency=false
      roundingType=FixedRadius
      width=0
      color {colorUseRule=UseColor_Specified, specifiedColor=<rgba 0 0 0 0>}
      cornerRadii=<12 12 12 12>
     }
     tintColor {colorUseRule=UseColor_Specified, specifiedColor=<rgba 1 1 1 1>}
     uVRegion {bIsValid=false, max=<0 0>, min=<0 0>}
    }
    renderTransform {angle=0, scale=<1 1>, shear=<0 0>, translation=<0 1>}
   }
   slot {class=OverlaySlot, properties {horizontalAlignment=HAlign_Fill, verticalAlignment=VAlign_Fill}}
  }
  FX_BackingOutline {
   widgetClassPath=&Image
   bIsVariable=true
   parent=ContainerFX
   widget=FX_BackingOutline
   properties {
    visibility=Collapsed
    brush {
     bIsDynamicallyLoaded=false
     drawAs=Box
     imageType=NoImage
     mirroring=NoMirror
     resourceName=None
     resourceObject=/TacOvershield/UI/Materials/M_UI_TacOvershield_Regen_Outline
     tiling=NoTile
     imageSize=<80 80>
     margin=<ltrb 0.5 0.5 0.5 0.5>
     outlineSettings {
      bUseBrushTransparency=false
      roundingType=FixedRadius
      width=0
      color {colorUseRule=UseColor_Specified, specifiedColor=<rgba 0 0 0 0>}
      cornerRadii=<12 12 12 12>
     }
     tintColor {colorUseRule=UseColor_Specified, specifiedColor=<rgba 1 1 1 1>}
     uVRegion {bIsValid=false, max=<0 0>, min=<0 0>}
    }
   }
   slot {class=OverlaySlot, properties {horizontalAlignment=HAlign_Fill, verticalAlignment=VAlign_Fill}}
  }
  FX_BackingGlow {
   widgetClassPath=&Image
   bIsVariable=true
   parent=ContainerFX
   widget=FX_BackingGlow
   properties {
    brush {
     bIsDynamicallyLoaded=false
     drawAs=Box
     imageType=NoImage
     mirroring=NoMirror
     resourceName=None
     resourceObject=/TacOvershield/UI/Materials/M_UI_TacOvershield_Regen_Glow
     tiling=NoTile
     imageSize=<80 80>
     margin=<ltrb 0.5 0.5 0.5 0.5>
     outlineSettings {
      bUseBrushTransparency=false
      roundingType=FixedRadius
      width=0
      color {colorUseRule=UseColor_Specified, specifiedColor=<rgba 0 0 0 0>}
      cornerRadii=<12 12 12 12>
     }
     tintColor {colorUseRule=UseColor_Specified, specifiedColor=<rgba 1 1 1 1>}
     uVRegion {bIsValid=false, max=<0 0>, min=<0 0>}
    }
    renderTransform {angle=0, scale=<1.0625 1.5>, shear=<0 0>, translation=<0 0>}
   }
   slot {class=OverlaySlot, properties {horizontalAlignment=HAlign_Fill, verticalAlignment=VAlign_Fill}}
  }
  FX_BackingGlow_Swipe {
   widgetClassPath=&Image
   bIsVariable=true
   parent=ContainerFX
   widget=FX_BackingGlow_Swipe
   properties {
    renderOpacity=0
    brush {
     bIsDynamicallyLoaded=false
     drawAs=Box
     imageType=NoImage
     mirroring=NoMirror
     resourceName=None
     resourceObject=/TacOvershield/UI/Materials/M_UI_TacOvershield_Regen_Glow_Swipe
     tiling=NoTile
     imageSize=<80 80>
     margin=<ltrb 0.5 0.5 0.5 0.5>
     outlineSettings {
      bUseBrushTransparency=false
      roundingType=FixedRadius
      width=0
      color {colorUseRule=UseColor_Specified, specifiedColor=<rgba 0 0 0 0>}
      cornerRadii=<12 12 12 12>
     }
     tintColor {colorUseRule=UseColor_Specified, specifiedColor=<rgba 1 1 1 1>}
     uVRegion {bIsValid=false, max=<0 0>, min=<0 0>}
    }
    renderTransform {angle=0, scale=<1.0625 1.5>, shear=<0 0>, translation=<0 0>}
   }
   slot {class=OverlaySlot, properties {horizontalAlignment=HAlign_Fill, verticalAlignment=VAlign_Fill}}
  }
  HorizontalBox_1 {
   widgetClassPath=&HorizontalBox
   parent=Overlay_1
   widget=HorizontalBox_1
   slot {class=OverlaySlot, properties {horizontalAlignment=HAlign_Fill, verticalAlignment=VAlign_Fill}}
  }
  Spacer_145 {
   widgetClassPath=&Spacer
   parent=HorizontalBox_1
   widget=Spacer_145
   slot {class=HorizontalBoxSlot, properties {size {sizeRule=Fill, value=0.825}}}
  }
  ScaleBox_1 {
   widgetClassPath=&ScaleBox
   parent=HorizontalBox_1
   widget=ScaleBox_1
   properties {stretch=ScaleToFitX}
   slot {class=HorizontalBoxSlot, properties {verticalAlignment=VAlign_Center, padding=<ltrb 10 0 10 0>, size {sizeRule=Fill, value=0.175}}}
  }
  HorizontalBox_0 {
   widgetClassPath=&HorizontalBox
   parent=ScaleBox_1
   widget=HorizontalBox_0
   slot {class=ScaleBoxSlot, properties {horizontalAlignment=HAlign_Fill, verticalAlignment=VAlign_Fill}}
  }
  AnimBoundIconBorder {
   widgetClassPath=&Border
   parent=HorizontalBox_0
   widget=AnimBoundIconBorder
   properties {
    background {
     bIsDynamicallyLoaded=false
     drawAs=NoDrawType
     imageType=NoImage
     mirroring=NoMirror
     resourceName=None
     resourceObject=null
     tiling=NoTile
     imageSize=<32 32>
     margin=<ltrb 0 0 0 0>
     outlineSettings {
      bUseBrushTransparency=false
      roundingType=HalfHeightRadius
      width=0
      color {colorUseRule=UseColor_Specified, specifiedColor=<rgba 0 0 0 0>}
      cornerRadii=<0 0 0 0>
     }
     tintColor {colorUseRule=UseColor_Specified, specifiedColor=<rgba 1 1 1 1>}
     uVRegion {bIsValid=false, max=<0 0>, min=<0 0>}
    }
    padding=<ltrb 0 0 0 0>
   }
   slot {class=HorizontalBoxSlot}
  }
  SizeBoxShieldProgress {
   widgetClassPath=&SizeBox
   parent=AnimBoundIconBorder
   widget=SizeBoxShieldProgress
   properties {
    bOverride_HeightOverride=true
    bOverride_WidthOverride=true
    heightOverride=56
    widthOverride=56
    renderTransform {angle=0, scale=<1 1>, shear=<0 0>, translation=<0 -2>}
   }
   slot {class=BorderSlot, properties {padding=<ltrb 0 0 0 0>}}
  }
  Overlay_2 {widgetClassPath=&Overlay, parent=SizeBoxShieldProgress, widget=Overlay_2, slot {class=SizeBoxSlot}}
  ShieldFullFX {
   widgetClassPath=&Overlay
   parent=Overlay_2
   widget=ShieldFullFX
   properties {renderOpacity=0}
   slot {class=OverlaySlot, properties {horizontalAlignment=HAlign_Center, verticalAlignment=VAlign_Center}}
  }
  ShieldAnimFX_Up {
   widgetClassPath=&Image
   bIsVariable=true
   parent=ShieldFullFX
   widget=ShieldAnimFX_Up
   properties {
    brush {
     bIsDynamicallyLoaded=false
     drawAs=Image
     imageType=NoImage
     mirroring=NoMirror
     resourceName=None
     resourceObject=/TacOvershield/UI/Materials/M_UI_TacOvershield_Pattern_Solid_Small_Shield
     tiling=NoTile
     imageSize=<48 48>
     margin=<ltrb 0 0 0 0>
     outlineSettings {
      bUseBrushTransparency=false
      roundingType=HalfHeightRadius
      width=0
      color {colorUseRule=UseColor_Specified, specifiedColor=<rgba 0 0 0 0>}
      cornerRadii=<0 0 0 0>
     }
     tintColor {colorUseRule=UseColor_Specified, specifiedColor=<rgba 1 1 1 1>}
     uVRegion {bIsValid=false, max=<0 0>, min=<0 0>}
    }
    renderTransform {angle=90, scale=<1.333 3>, shear=<0 0>, translation=<0 0>}
   }
   slot {class=OverlaySlot, properties {horizontalAlignment=HAlign_Center, verticalAlignment=VAlign_Center, padding=<ltrb 0 0 0 40>}}
  }
  ShieldAnimFX_Down {
   widgetClassPath=&Image
   bIsVariable=true
   parent=ShieldFullFX
   widget=ShieldAnimFX_Down
   properties {
    brush {
     bIsDynamicallyLoaded=false
     drawAs=Image
     imageType=NoImage
     mirroring=NoMirror
     resourceName=None
     resourceObject=/TacOvershield/UI/Materials/M_UI_TacOvershield_Pattern_Solid_Small_Shield_2
     tiling=NoTile
     imageSize=<48 48>
     margin=<ltrb 0 0 0 0>
     outlineSettings {
      bUseBrushTransparency=false
      roundingType=HalfHeightRadius
      width=0
      color {colorUseRule=UseColor_Specified, specifiedColor=<rgba 0 0 0 0>}
      cornerRadii=<0 0 0 0>
     }
     tintColor {colorUseRule=UseColor_Specified, specifiedColor=<rgba 1 1 1 1>}
     uVRegion {bIsValid=false, max=<0 0>, min=<0 0>}
    }
    renderTransform {angle=-90, scale=<1.333 2>, shear=<0 0>, translation=<0 0>}
   }
   slot {class=OverlaySlot, properties {horizontalAlignment=HAlign_Center, verticalAlignment=VAlign_Center, padding=<ltrb 0 32 0 0>}}
  }
  Flares_1 {
   widgetClassPath=&Overlay
   parent=Overlay_2
   widget=Flares_1
   properties {renderOpacity=0, visibility=Collapsed}
   slot {class=OverlaySlot, properties {horizontalAlignment=HAlign_Center, verticalAlignment=VAlign_Fill}}
  }
  Flare {
   widgetClassPath=&Image
   bIsVariable=true
   parent=Flares_1
   widget=Flare
   properties {
    renderOpacity=0.0125
    brush {
     bIsDynamicallyLoaded=false
     drawAs=Image
     imageType=NoImage
     mirroring=NoMirror
     resourceName=None
     resourceObject=/Game/Athena/HUD/XP/Materials/M_UI_AnimatedLevelUpGlow
     tiling=NoTile
     imageSize=<4 32>
     margin=<ltrb 0 0 0 0>
     outlineSettings {
      bUseBrushTransparency=false
      roundingType=HalfHeightRadius
      width=0
      color {colorUseRule=UseColor_Specified, specifiedColor=<rgba 0 0 0 0>}
      cornerRadii=<0 0 0 0>
     }
     tintColor {colorUseRule=UseColor_Specified, specifiedColor=<rgba 1 1 1 1>}
     uVRegion {bIsValid=false, max=<0 0>, min=<0 0>}
    }
    renderTransform {angle=65, scale=<1 3.5>, shear=<0 0>, translation=<0 0>}
   }
   slot {class=OverlaySlot, properties {horizontalAlignment=HAlign_Center, verticalAlignment=VAlign_Fill}}
  }
  Flare_3 {
   widgetClassPath=&Image
   bIsVariable=true
   parent=Flares_1
   widget=Flare_3
   properties {
    renderOpacity=0.0625
    brush {
     bIsDynamicallyLoaded=false
     drawAs=Image
     imageType=NoImage
     mirroring=NoMirror
     resourceName=None
     resourceObject=&M_UI_AnimatedLevelUpGlow
     tiling=NoTile
     imageSize=<8 32>
     margin=<ltrb 0 0 0 0>
     outlineSettings {
      bUseBrushTransparency=false
      roundingType=HalfHeightRadius
      width=0
      color {colorUseRule=UseColor_Specified, specifiedColor=<rgba 0 0 0 0>}
      cornerRadii=<0 0 0 0>
     }
     tintColor {colorUseRule=UseColor_Specified, specifiedColor=<rgba 1 1 1 1>}
     uVRegion {bIsValid=false, max=<0 0>, min=<0 0>}
    }
    renderTransform {angle=33, scale=<1 5>, shear=<0 0>, translation=<0 0>}
   }
   slot {class=OverlaySlot, properties {horizontalAlignment=HAlign_Center, verticalAlignment=VAlign_Fill}}
  }
  RoundGlow_1 {
   widgetClassPath=&Image
   bIsVariable=true
   parent=Overlay_2
   widget=RoundGlow_1
   properties {
    clipping=ClipToBoundsWithoutIntersecting
    renderOpacity=0
    brush {
     bIsDynamicallyLoaded=false
     drawAs=Image
     imageType=NoImage
     mirroring=NoMirror
     resourceName=None
     resourceObject=/Game/Athena/HUD/LTM/Wax/Materials/M_UI_OS_Icon_Glow
     tiling=NoTile
     imageSize=<32 32>
     margin=<ltrb 0 0 0 0>
     outlineSettings {
      bUseBrushTransparency=false
      roundingType=HalfHeightRadius
      width=0
      color {colorUseRule=UseColor_Specified, specifiedColor=<rgba 0 0 0 0>}
      cornerRadii=<0 0 0 0>
     }
     tintColor {colorUseRule=UseColor_Specified, specifiedColor=<rgba 1 1 1 1>}
     uVRegion {bIsValid=false, max=<0 0>, min=<0 0>}
    }
    renderTransform {angle=0, scale=<3.25 3.25>, shear=<0 0>, translation=<0 0>}
   }
   slot {class=OverlaySlot, properties {horizontalAlignment=HAlign_Fill, verticalAlignment=VAlign_Fill}}
  }
  NoiseGlow_1 {
   widgetClassPath=&Image
   bIsVariable=true
   parent=Overlay_2
   widget=NoiseGlow_1
   properties {
    renderOpacity=0
    brush {
     bIsDynamicallyLoaded=false
     drawAs=Image
     imageType=NoImage
     mirroring=NoMirror
     resourceName=None
     resourceObject=/Game/Athena/HUD/XP/Materials/M_UI_LevelUp_LightShaft_Inst
     tiling=NoTile
     imageSize=<32 32>
     margin=<ltrb 0 0 0 0>
     outlineSettings {
      bUseBrushTransparency=false
      roundingType=HalfHeightRadius
      width=0
      color {colorUseRule=UseColor_Specified, specifiedColor=<rgba 0 0 0 0>}
      cornerRadii=<0 0 0 0>
     }
     tintColor {colorUseRule=UseColor_Specified, specifiedColor=<rgba 1 1 1 1>}
     uVRegion {bIsValid=false, max=<0 0>, min=<0 0>}
    }
    renderTransform {angle=0, scale=<5 5>, shear=<0 0>, translation=<0 0>}
   }
   slot {class=OverlaySlot, properties {horizontalAlignment=HAlign_Fill, verticalAlignment=VAlign_Fill}}
  }
  Icon_ShieldAnim_1 {
   widgetClassPath=&Image
   bIsVariable=true
   parent=Overlay_2
   widget=Icon_ShieldAnim_1
   properties {
    clipping=ClipToBoundsWithoutIntersecting
    renderOpacity=0
    brush {
     bIsDynamicallyLoaded=false
     drawAs=Image
     imageType=NoImage
     mirroring=NoMirror
     resourceName=None
     resourceObject=/TacOvershield/UI/Materials/M_UI_TacticalOvershield_Pulse
     tiling=NoTile
     imageSize=<56 56>
     margin=<ltrb 0 0 0 0>
     outlineSettings {
      bUseBrushTransparency=false
      roundingType=HalfHeightRadius
      width=0
      color {colorUseRule=UseColor_Specified, specifiedColor=<rgba 0 0 0 0>}
      cornerRadii=<0 0 0 0>
     }
     tintColor {colorUseRule=UseColor_Specified, specifiedColor=<rgba 1 1 1 1>}
     uVRegion {bIsValid=false, max=<0 0>, min=<0 0>}
    }
   }
   slot {class=OverlaySlot, properties {horizontalAlignment=HAlign_Center, verticalAlignment=VAlign_Center}}
  }
  Image_ShieldFull {
   widgetClassPath=&Image
   bIsVariable=true
   parent=Overlay_2
   widget=Image_ShieldFull
   properties {
    renderOpacity=0
    brush {
     bIsDynamicallyLoaded=false
     drawAs=Image
     imageType=NoImage
     mirroring=NoMirror
     resourceName=None
     resourceObject=/TacOvershield/UI/Materials/M_UI_Overshield_Icon_ProgressBar_FullIndication
     tiling=NoTile
     imageSize=<32 32>
     margin=<ltrb 0 0 0 0>
     outlineSettings {
      bUseBrushTransparency=false
      roundingType=HalfHeightRadius
      width=0
      color {colorUseRule=UseColor_Specified, specifiedColor=<rgba 0 0 0 0>}
      cornerRadii=<0 0 0 0>
     }
     tintColor {colorUseRule=UseColor_Specified, specifiedColor=<rgba 1 1 1 1>}
     uVRegion {bIsValid=false, max=<0 0>, min=<0 0>}
    }
    renderTransform {angle=0, scale=<1.25 1.25>, shear=<0 0>, translation=<0 0>}
   }
   slot {class=OverlaySlot, properties {horizontalAlignment=HAlign_Fill, verticalAlignment=VAlign_Fill}}
  }
  Icon_Delta {
   widgetClassPath=&Image
   bIsVariable=true
   parent=Overlay_2
   widget=Icon_Delta
   properties {brush {
     bIsDynamicallyLoaded=false
     drawAs=Image
     imageType=NoImage
     mirroring=NoMirror
     resourceName=None
     resourceObject=/TacOvershield/UI/Materials/M_UI_Overshield_Icon_ProgressBar_Delta
     tiling=NoTile
     imageSize=<56 56>
     margin=<ltrb 0 0 0 0>
     outlineSettings {
      bUseBrushTransparency=false
      roundingType=HalfHeightRadius
      width=0
      color {colorUseRule=UseColor_Specified, specifiedColor=<rgba 0 0 0 0>}
      cornerRadii=<0 0 0 0>
     }
     tintColor {colorUseRule=UseColor_Specified, specifiedColor=<rgba 1 1 1 1>}
     uVRegion {bIsValid=false, max=<0 0>, min=<0 0>}
    }}
   slot {class=OverlaySlot, properties {horizontalAlignment=HAlign_Fill, verticalAlignment=VAlign_Fill}}
  }
  Icon_Progress {
   widgetClassPath=&Image
   bIsVariable=true
   parent=Overlay_2
   widget=Icon_Progress
   properties {brush {
     bIsDynamicallyLoaded=false
     drawAs=Image
     imageType=NoImage
     mirroring=NoMirror
     resourceName=None
     resourceObject=/TacOvershield/UI/Materials/M_UI_Overshield_Icon_ProgressBar_Fill
     tiling=NoTile
     imageSize=<56 56>
     margin=<ltrb 0 0 0 0>
     outlineSettings {
      bUseBrushTransparency=false
      roundingType=HalfHeightRadius
      width=0
      color {colorUseRule=UseColor_Specified, specifiedColor=<rgba 0 0 0 0>}
      cornerRadii=<0 0 0 0>
     }
     tintColor {colorUseRule=UseColor_Specified, specifiedColor=<rgba 1 1 1 1>}
     uVRegion {bIsValid=false, max=<0 0>, min=<0 0>}
    }}
   slot {class=OverlaySlot, properties {horizontalAlignment=HAlign_Fill, verticalAlignment=VAlign_Fill}}
  }
  Icon_Broken_1 {
   widgetClassPath=&Image
   bIsVariable=true
   parent=Overlay_2
   widget=Icon_Broken_1
   properties {brush {
     bIsDynamicallyLoaded=false
     drawAs=Image
     imageType=NoImage
     mirroring=NoMirror
     resourceName=None
     resourceObject=/TacOvershield/UI/Materials/M_UI_TacticalOvershield_Broken_Icon
     tiling=NoTile
     imageSize=<56 56>
     margin=<ltrb 0 0 0 0>
     outlineSettings {
      bUseBrushTransparency=false
      roundingType=HalfHeightRadius
      width=0
      color {colorUseRule=UseColor_Specified, specifiedColor=<rgba 0 0 0 0>}
      cornerRadii=<0 0 0 0>
     }
     tintColor {colorUseRule=UseColor_Specified, specifiedColor=<rgba 1 1 1 1>}
     uVRegion {bIsValid=false, max=<0 0>, min=<0 0>}
    }}
   slot {class=OverlaySlot, properties {horizontalAlignment=HAlign_Fill, verticalAlignment=VAlign_Fill}}
  }
  Icon_Broken_Shards_1 {
   widgetClassPath=&Image
   bIsVariable=true
   parent=Overlay_2
   widget=Icon_Broken_Shards_1
   properties {
    visibility=SelfHitTestInvisible
    brush {
     bIsDynamicallyLoaded=false
     drawAs=Image
     imageType=NoImage
     mirroring=NoMirror
     resourceName=None
     resourceObject=/TacOvershield/UI/Materials/M_UI_Overshield_CrackedFX
     tiling=NoTile
     imageSize=<32 32>
     margin=<ltrb 0 0 0 0>
     outlineSettings {
      bUseBrushTransparency=false
      roundingType=HalfHeightRadius
      width=0
      color {colorUseRule=UseColor_Specified, specifiedColor=<rgba 0 0 0 0>}
      cornerRadii=<0 0 0 0>
     }
     tintColor {colorUseRule=UseColor_Specified, specifiedColor=<rgba 1 1 1 1>}
     uVRegion {bIsValid=false, max=<0 0>, min=<0 0>}
    }
    renderTransform {angle=0, scale=<4 4>, shear=<0 0>, translation=<-4 0>}
   }
   slot {class=OverlaySlot, properties {horizontalAlignment=HAlign_Fill, verticalAlignment=VAlign_Fill}}
  }
  SizeBoxTris_1 {
   widgetClassPath=&SizeBox
   parent=Overlay_2
   widget=SizeBoxTris_1
   properties {
    bOverride_HeightOverride=true
    bOverride_WidthOverride=true
    heightOverride=44
    widthOverride=44
    renderTransform {angle=20, scale=<1 1>, shear=<0 0>, translation=<0 0>}
   }
   slot {class=OverlaySlot, properties {horizontalAlignment=HAlign_Center, verticalAlignment=VAlign_Center}}
  }
  OverlayTris2 {
   widgetClassPath=&Overlay
   bIsVariable=true
   parent=SizeBoxTris_1
   widget=OverlayTris2
   properties {renderOpacity=0, renderTransform {angle=0, scale=<0 0>, shear=<0 0>, translation=<0 0>}}
   slot {class=SizeBoxSlot}
  }
  TriFX {
   widgetClassPath=&Image
   bIsVariable=true
   parent=OverlayTris2
   widget=TriFX
   properties {
    renderOpacity=0.8
    visibility=SelfHitTestInvisible
    brush {
     bIsDynamicallyLoaded=false
     drawAs=Image
     imageType=NoImage
     mirroring=NoMirror
     resourceName=None
     resourceObject=/FortUILibrary/MaterialInstances/Effects/M_UI_Tri_FX_HUDLevelup
     tiling=NoTile
     imageSize=<32 16>
     margin=<ltrb 0 0 0 0>
     outlineSettings {
      bUseBrushTransparency=false
      roundingType=HalfHeightRadius
      width=0
      color {colorUseRule=UseColor_Specified, specifiedColor=<rgba 0 0 0 0>}
      cornerRadii=<0 0 0 0>
     }
     tintColor {colorUseRule=UseColor_Specified, specifiedColor=<rgba 1 1 1 1>}
     uVRegion {bIsValid=false, max=<0 0>, min=<0 0>}
    }
    renderTransform {angle=90, scale=<2 2>, shear=<0 0>, translation=<22 0>}
    renderTransformPivot=<1 0>
   }
   slot {class=OverlaySlot, properties {horizontalAlignment=HAlign_Fill, verticalAlignment=VAlign_Fill}}
  }
  TriFX_3 {
   widgetClassPath=&Image
   bIsVariable=true
   parent=OverlayTris2
   widget=TriFX_3
   properties {
    renderOpacity=0.8
    visibility=SelfHitTestInvisible
    brush {
     bIsDynamicallyLoaded=false
     drawAs=Image
     imageType=NoImage
     mirroring=NoMirror
     resourceName=None
     resourceObject=&M_UI_Tri_FX_HUDLevelup
     tiling=NoTile
     imageSize=<32 16>
     margin=<ltrb 0 0 0 0>
     outlineSettings {
      bUseBrushTransparency=false
      roundingType=HalfHeightRadius
      width=0
      color {colorUseRule=UseColor_Specified, specifiedColor=<rgba 0 0 0 0>}
      cornerRadii=<0 0 0 0>
     }
     tintColor {colorUseRule=UseColor_Specified, specifiedColor=<rgba 1 1 1 1>}
     uVRegion {bIsValid=false, max=<0 0>, min=<0 0>}
    }
    renderTransform {angle=-90, scale=<2 2>, shear=<0 0>, translation=<22 0>}
    renderTransformPivot=<1 1>
   }
   slot {class=OverlaySlot, properties {horizontalAlignment=HAlign_Fill, verticalAlignment=VAlign_Fill}}
  }
  SizeBox_0 {
   widgetClassPath=&SizeBox
   parent=HorizontalBox_0
   widget=SizeBox_0
   properties {bOverride_MinDesiredWidth=true, minDesiredWidth=35}
   slot {class=HorizontalBoxSlot, properties {horizontalAlignment=HAlign_Left, verticalAlignment=VAlign_Center, padding=<ltrb 4 0 4 0>}}
  }
  AnimBoundColorChange {
   widgetClassPath=&Border
   parent=SizeBox_0
   widget=AnimBoundColorChange
   properties {
    horizontalAlignment=HAlign_Left
    background {
     bIsDynamicallyLoaded=false
     drawAs=NoDrawType
     imageType=NoImage
     mirroring=NoMirror
     resourceName=None
     resourceObject=null
     tiling=NoTile
     imageSize=<32 32>
     margin=<ltrb 0 0 0 0>
     outlineSettings {
      bUseBrushTransparency=false
      roundingType=HalfHeightRadius
      width=0
      color {colorUseRule=UseColor_Specified, specifiedColor=<rgba 0 0 0 0>}
      cornerRadii=<0 0 0 0>
     }
     tintColor {colorUseRule=UseColor_Specified, specifiedColor=<rgba 1 1 1 1>}
     uVRegion {bIsValid=false, max=<0 0>, min=<0 0>}
    }
    padding=<ltrb 0 0 0 0>
   }
   slot {class=SizeBoxSlot, properties {horizontalAlignment=HAlign_Left}}
  }
  Overlay_0 {
   widgetClassPath=&Overlay
   parent=AnimBoundColorChange
   widget=Overlay_0
   slot {class=BorderSlot, properties {horizontalAlignment=HAlign_Left, padding=<ltrb 0 0 0 0>}}
  }
  CurrentValue {
   widgetClassPath=&CommonTextBlock
   bIsVariable=true
   parent=Overlay_0
   widget=CurrentValue
   properties {
    lineHeightPercentage=0.8
    text=99
    font {
     bForceMonospaced=false
     bMaterialIsStencil=false
     fontMaterial=null
     fontObject=&BurbankSmall
     letterSpacing=0
     monospacedWidth=1
     size=15
     skewAmount=0
     typefaceFontName=Black
     outlineSettings {
      bApplyOutlineToDropShadows=true
      bMiteredCorners=false
      bSeparateFillAlpha=true
      outlineMaterial=&M-FontShadow
      outlineSize=2
      outlineColor=<rgba 0 0 0 0.5>
     }
    }
    margin=<ltrb 0 -6 0 1>
    renderTransformPivot=<0 0.75>
    shadowOffset=<0 0>
   }
   slot {class=OverlaySlot}
  }
  TutorialHighlight {widgetClassPath=&TutorialHighlight}
 ]
}
