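# Last-killer-info HUD extension widget that shows the killer's Tactical Overshield as an icon and number, collapsed on construct unless the feature is enabled.
# SetKillerHealthAndShield passes only OvershieldPercent to UpdateFillBar and UpdateValue.
# UpdateFillBar drives the icon material's Progress_Amount and swaps in a broken icon at zero or below.
# UpdateValue displays percent times 50, rounded; its show-1 branch never fires because it tests constant 0 > 0.
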WidgetBlueprint /TacOvershield/UI/TacOverShieldWidget_LastKillerInfo
graphs [
 EventGraph {nodes [
   Event PreConstruct()
   > n1 = GetDynamicMaterial(self=Get Icon_Delta() -> (Icon_Delta:Image Object Reference))
   > SetScalarParameterValue(self=@n1, ParameterName=Progress_Amount, Value=0)
   Event Construct()
   > n2 = IsTacticalOvershieldEnabled!(execute) -> (then, ReturnValue:bool)
   > n3 = Branch(Condition=@n2)
   > SetVisibility(InVisibility=HitTestInvisible)
   SetVisibility(execute=@n3.else, InVisibility=Collapsed)
   n4 = Event SetKillerHealthAndShield!() -> (OutputDelegate, then, HealthPercent:float, ShieldPercent:float, OvershieldPercent:float) {is=title K2Node_Event}
   > UpdateFillBar!(execute, self, Value:double=@n4.OvershieldPercent) -> (then)
   > UpdateValue!(execute, self, Value:double=@n4.OvershieldPercent) -> (then)
  ]}
 UpdateFillBar {nodes [
   n1 = UpdateFillBar!() -> (then, Value:double=~) {is=title K2Node_FunctionEntry}
   > n2 = GetDynamicMaterial(self=Get Icon_Progress() -> (Icon_Progress:Image Object Reference))
   > SetScalarParameterValue(self=@n2, ParameterName=Progress_Amount, Value=@n1)
   > UpdateShieldBrokenStatus!(execute, self, bShieldBroken:bool="float <= float"(A=@n1) -> (ReturnValue=~)) -> (then)
  ]}
 UpdateShieldBrokenStatus {nodes [
   n1 = UpdateShieldBrokenStatus!() -> (then, bShieldBroken:bool=~) {is=title K2Node_FunctionEntry}
   > SetVisibility(
    self=Get Icon_Broken_1() -> (Icon_Broken_1:Image Object Reference)
    InVisibility=Select(Option 0=Collapsed, Option 1=HitTestInvisible, Index=@n1)
   )
   > SetVisibility(self=[Get Icon_Delta(), Get Icon_Progress()], InVisibility=Select(Option 0=HitTestInvisible, Option 1=Collapsed, Index=@n1))
  ]}
 UpdateValue {nodes [
   n1 = UpdateValue!() -> (then, Value:double=~) {is=title K2Node_FunctionEntry}
   n2 = Round(A=float * float(A=@n1, B=50) {is=title K2Node_PromotableOperator}) {is=title K2Node_CallFunction}
   > n3 = Branch(Condition=AND Boolean(A="Equal (Integer)"(A=@n2, B=0) {is=title K2Node_CallFunction}, B="float > float"(A=0, B=0) {is=title K2Node_CallFunction}))
   n4 = Get CurrentValue() -> (CurrentValue:Common Text Object Reference)
   > "SetText (Text)"(self=@n4, InText="To Text (Integer)"(Value=1))
   "SetText (Text)"(execute=@n3.else, self=@n4, InText="To Text (Integer)"(Value=@n2))
  ]}
]
variables [
 ActiveShieldIndex {type=integer, default=1}
 bBarIsFull {type=boolean}
 bIsBroken {type=boolean, default=true}
 bRegen {type=boolean}
 BrokenStyle {
  type=ProgressBarStyle
  default {
   enableFillAnimation=false
   backgroundImage {
    bIsDynamicallyLoaded=false
    drawAs=Box
    imageType=FullColor
    mirroring=NoMirror
    resourceName=../../../Engine/Content/Slate/Common/ProgressBar_Background.png
    resourceObject=null
    tiling=NoTile
    imageSize=<0 0>
    margin=<ltrb 0.4167 0.4167 0.4167 0.4167>
    outlineSettings {
     bUseBrushTransparency=false
     roundingType=HalfHeightRadius
     width=0
     color {colorUseRule=UseColor_Specified, specifiedColor=<rgba 0 0 0 0>}
     cornerRadii=<0 0 0 1>
    }
    tintColor {colorUseRule=UseColor_Specified, specifiedColor=<rgba 1 0 0 0.5>}
    uVRegion {bIsValid=false, max=<0 0>, min=<0 0>}
   }
   fillImage {
    bIsDynamicallyLoaded=false
    drawAs=Box
    imageType=FullColor
    mirroring=NoMirror
    resourceName=../../../Engine/Content/Slate/Common/ProgressBar_Fill.png
    resourceObject=null
    tiling=NoTile
    imageSize=<0 0>
    margin=<ltrb 0.4167 0.4167 0.4167 0.4167>
    outlineSettings {
     bUseBrushTransparency=false
     roundingType=HalfHeightRadius
     width=0
     color {colorUseRule=UseColor_Specified, specifiedColor=<rgba 0 0 0 0>}
     cornerRadii=<0 0 0 1>
    }
    tintColor {colorUseRule=UseColor_Specified, specifiedColor=<rgba 1 1 1 0.85>}
    uVRegion {bIsValid=false, max=<0 0>, min=<0 0>}
   }
   marqueeImage {
    bIsDynamicallyLoaded=false
    drawAs=Image
    imageType=FullColor
    mirroring=NoMirror
    resourceName=../../../Engine/Content/Slate/Common/ProgressBar_Marquee.png
    resourceObject=null
    tiling=Horizontal
    imageSize=<20 12>
    margin=<ltrb 0 0 0 0>
    outlineSettings {
     bUseBrushTransparency=false
     roundingType=HalfHeightRadius
     width=0
     color {colorUseRule=UseColor_Specified, specifiedColor=<rgba 0 0 0 0>}
     cornerRadii=<0 0 0 1>
    }
    tintColor {colorUseRule=UseColor_Specified, specifiedColor=<rgba 1 1 1 1>}
    uVRegion {bIsValid=false, max=<0 0>, min=<0 0>}
   }
  }
 }
 bUseIconVersion {type=boolean, default=true}
 DefaultStyle {
  type=ProgressBarStyle
  default {
   enableFillAnimation=false
   backgroundImage {
    bIsDynamicallyLoaded=false
    drawAs=Box
    imageType=FullColor
    mirroring=NoMirror
    resourceName=../../../Engine/Content/Slate/Common/ProgressBar_Background.png
    resourceObject=null
    tiling=NoTile
    imageSize=<0 0>
    margin=<ltrb 0.4167 0.4167 0.4167 0.4167>
    outlineSettings {
     bUseBrushTransparency=false
     roundingType=HalfHeightRadius
     width=0
     color {colorUseRule=UseColor_Specified, specifiedColor=<rgba 0 0 0 0>}
     cornerRadii=<0 0 0 1>
    }
    tintColor {colorUseRule=UseColor_Specified, specifiedColor=<rgba 0 0 0 0.25>}
    uVRegion {bIsValid=false, max=<0 0>, min=<0 0>}
   }
   fillImage {
    bIsDynamicallyLoaded=false
    drawAs=Box
    imageType=FullColor
    mirroring=NoMirror
    resourceName=../../../Engine/Content/Slate/Common/ProgressBar_Fill.png
    resourceObject=null
    tiling=NoTile
    imageSize=<0 0>
    margin=<ltrb 0.4167 0.4167 0.4167 0.4167>
    outlineSettings {
     bUseBrushTransparency=false
     roundingType=HalfHeightRadius
     width=0
     color {colorUseRule=UseColor_Specified, specifiedColor=<rgba 0 0 0 0>}
     cornerRadii=<0 0 0 1>
    }
    tintColor {colorUseRule=UseColor_Specified, specifiedColor=<rgba 1 1 1 0.85>}
    uVRegion {bIsValid=false, max=<0 0>, min=<0 0>}
   }
   marqueeImage {
    bIsDynamicallyLoaded=false
    drawAs=Image
    imageType=FullColor
    mirroring=NoMirror
    resourceName=../../../Engine/Content/Slate/Common/ProgressBar_Marquee.png
    resourceObject=null
    tiling=Horizontal
    imageSize=<20 12>
    margin=<ltrb 0 0 0 0>
    outlineSettings {
     bUseBrushTransparency=false
     roundingType=HalfHeightRadius
     width=0
     color {colorUseRule=UseColor_Specified, specifiedColor=<rgba 0 0 0 0>}
     cornerRadii=<0 0 0 1>
    }
    tintColor {colorUseRule=UseColor_Specified, specifiedColor=<rgba 1 1 1 1>}
    uVRegion {bIsValid=false, max=<0 0>, min=<0 0>}
   }
  }
 }
 HideValueTimerHandle {type=TimerHandle}
]
widgets {
 info {
  inheritedWidgetCount=1
  namedSlotCount=1
  parentClass=/Script/FortniteUI.LastKillerInfoExtensionWidget
  rootWidgetClass=/Game/Athena/HUD/FortHUDElementWrapper.FortHUDElementWrapper_C
  widgetCount=14
 }
 widgets [
  FortHUDElementWrapper {
   widgetClassPath=/Game/Athena/HUD/FortHUDElementWrapper
   bIsVariable=true
   widget=FortHUDElementWrapper
   properties {
    elementTag {tagName=HUD.Athena.Player.Overshield.LastKillerInfo}
    hUDElementTag {gameplayTags [
      {tagName=Athena.HideForCapture}
      {tagName=HUD.Athena.HideforRadial}
      {tagName=HUD.Event.HideForSpecialEvent}
      {tagName=HUD.Athena.Hide.WithInventory}
      {tagName=HUD.Hide.WhileLoading}
      {tagName=HUD.Athena.Hide.WithMap}
      {tagName=Athena.Creative.HideForLoading}
      {tagName=Athena.Creative.HUDController.HideShields}
      {tagName=HUD.Hide.CreativeJoiningLobby}
     ]}
   }
  }
  Overlay_1 {widgetClassPath=&Overlay, namedSlotHost=FortHUDElementWrapper, widget=Overlay_1}
  FullSizeBox {
   widgetClassPath=&SizeBox
   bIsVariable=true
   parent=Overlay_1
   widget=FullSizeBox
   properties {bOverride_HeightOverride=true, bOverride_WidthOverride=true, heightOverride=80, maxDesiredWidth=570, widthOverride=540}
   slot {class=OverlaySlot, properties {verticalAlignment=VAlign_Fill}}
  }
  ShieldContainer {
   widgetClassPath=&Image
   bIsVariable=true
   parent=FullSizeBox
   widget=ShieldContainer
   properties {brush {
     bIsDynamicallyLoaded=false
     drawAs=Box
     imageType=NoImage
     mirroring=NoMirror
     resourceName=None
     resourceObject=/TacOvershield/UI/Materials/M_UI_TacOvershield_Backing
     tiling=NoTile
     imageSize=<80 80>
     margin=<ltrb 0.5 0.5 0.5 0.5>
     outlineSettings {
      bUseBrushTransparency=false
      roundingType=FixedRadius
      width=0
      color {colorUseRule=UseColor_Specified, specifiedColor=<rgba 0 0 0 0>}
      cornerRadii=<12 12 12 12>
     }
     tintColor {colorUseRule=UseColor_Specified, specifiedColor=<rgba 1 1 1 1>}
     uVRegion {bIsValid=false, max=<0 0>, min=<0 0>}
    }}
   slot {class=SizeBoxSlot}
  }
  HorizontalBox_0 {
   widgetClassPath=&HorizontalBox
   parent=Overlay_1
   widget=HorizontalBox_0
   slot {class=OverlaySlot, properties {horizontalAlignment=HAlign_Right, verticalAlignment=VAlign_Center, padding=<ltrb 0 0 2 0>}}
  }
  SizeBoxShieldProgress {
   widgetClassPath=&SizeBox
   parent=HorizontalBox_0
   widget=SizeBoxShieldProgress
   properties {bOverride_HeightOverride=true, bOverride_WidthOverride=true, heightOverride=56, widthOverride=56}
   slot {class=HorizontalBoxSlot}
  }
  Overlay_2 {widgetClassPath=&Overlay, parent=SizeBoxShieldProgress, widget=Overlay_2, slot {class=SizeBoxSlot}}
  Icon_Delta {
   widgetClassPath=&Image
   bIsVariable=true
   parent=Overlay_2
   widget=Icon_Delta
   properties {brush {
     bIsDynamicallyLoaded=false
     drawAs=Image
     imageType=NoImage
     mirroring=NoMirror
     resourceName=None
     resourceObject=/TacOvershield/UI/Materials/M_UI_Overshield_Icon_ProgressBar_Delta
     tiling=NoTile
     imageSize=<56 56>
     margin=<ltrb 0 0 0 0>
     outlineSettings {
      bUseBrushTransparency=false
      roundingType=HalfHeightRadius
      width=0
      color {colorUseRule=UseColor_Specified, specifiedColor=<rgba 0 0 0 0>}
      cornerRadii=<0 0 0 0>
     }
     tintColor {colorUseRule=UseColor_Specified, specifiedColor=<rgba 1 1 1 1>}
     uVRegion {bIsValid=false, max=<0 0>, min=<0 0>}
    }}
   slot {class=OverlaySlot, properties {horizontalAlignment=HAlign_Fill, verticalAlignment=VAlign_Fill}}
  }
  Icon_Progress {
   widgetClassPath=&Image
   bIsVariable=true
   parent=Overlay_2
   widget=Icon_Progress
   properties {brush {
     bIsDynamicallyLoaded=false
     drawAs=Image
     imageType=NoImage
     mirroring=NoMirror
     resourceName=None
     resourceObject=/TacOvershield/UI/Materials/M_UI_Overshield_Icon_ProgressBar_Fill
     tiling=NoTile
     imageSize=<56 56>
     margin=<ltrb 0 0 0 0>
     outlineSettings {
      bUseBrushTransparency=false
      roundingType=HalfHeightRadius
      width=0
      color {colorUseRule=UseColor_Specified, specifiedColor=<rgba 0 0 0 0>}
      cornerRadii=<0 0 0 0>
     }
     tintColor {colorUseRule=UseColor_Specified, specifiedColor=<rgba 1 1 1 1>}
     uVRegion {bIsValid=false, max=<0 0>, min=<0 0>}
    }}
   slot {class=OverlaySlot, properties {horizontalAlignment=HAlign_Fill, verticalAlignment=VAlign_Fill}}
  }
  Icon_Broken_1 {
   widgetClassPath=&Image
   bIsVariable=true
   parent=Overlay_2
   widget=Icon_Broken_1
   properties {
    renderOpacity=0.25
    visibility=Collapsed
    brush {
     bIsDynamicallyLoaded=false
     drawAs=Image
     imageType=NoImage
     mirroring=NoMirror
     resourceName=None
     resourceObject=/TacOvershield/UI/Materials/M_UI_TacticalOvershield_Broken_Icon
     tiling=NoTile
     imageSize=<44 44>
     margin=<ltrb 0 0 0 0>
     outlineSettings {
      bUseBrushTransparency=false
      roundingType=HalfHeightRadius
      width=0
      color {colorUseRule=UseColor_Specified, specifiedColor=<rgba 0 0 0 0>}
      cornerRadii=<0 0 0 0>
     }
     tintColor {colorUseRule=UseColor_Specified, specifiedColor=<rgba 1 1 1 1>}
     uVRegion {bIsValid=false, max=<0 0>, min=<0 0>}
    }
   }
   slot {class=OverlaySlot, properties {horizontalAlignment=HAlign_Fill, verticalAlignment=VAlign_Fill}}
  }
  SizeBox_0 {
   widgetClassPath=&SizeBox
   parent=HorizontalBox_0
   widget=SizeBox_0
   properties {bOverride_MinDesiredWidth=true, minDesiredWidth=35}
   slot {class=HorizontalBoxSlot, properties {verticalAlignment=VAlign_Center, padding=<ltrb 4 0 4 0>, size {sizeRule=Fill, value=1}}}
  }
  AnimBoundColorChange {
   widgetClassPath=&Border
   parent=SizeBox_0
   widget=AnimBoundColorChange
   properties {
    background {
     bIsDynamicallyLoaded=false
     drawAs=NoDrawType
     imageType=NoImage
     mirroring=NoMirror
     resourceName=None
     resourceObject=null
     tiling=NoTile
     imageSize=<32 32>
     margin=<ltrb 0 0 0 0>
     outlineSettings {
      bUseBrushTransparency=false
      roundingType=HalfHeightRadius
      width=0
      color {colorUseRule=UseColor_Specified, specifiedColor=<rgba 0 0 0 0>}
      cornerRadii=<0 0 0 0>
     }
     tintColor {colorUseRule=UseColor_Specified, specifiedColor=<rgba 1 1 1 1>}
     uVRegion {bIsValid=false, max=<0 0>, min=<0 0>}
    }
    padding=<ltrb 0 0 0 0>
   }
   slot {class=SizeBoxSlot}
  }
  CurrentValue {
   widgetClassPath=&CommonTextBlock
   bIsVariable=true
   parent=AnimBoundColorChange
   widget=CurrentValue
   properties {
    justification=Center
    lineHeightPercentage=0.8
    text=99
    font {
     bForceMonospaced=false
     bMaterialIsStencil=false
     fontMaterial=null
     fontObject=&BurbankSmall
     letterSpacing=0
     monospacedWidth=1
     size=15
     skewAmount=0
     typefaceFontName=Black
     outlineSettings {
      bApplyOutlineToDropShadows=true
      bMiteredCorners=false
      bSeparateFillAlpha=true
      outlineMaterial=&M-FontShadow
      outlineSize=2
      outlineColor=<rgba 0 0 0 0.5>
     }
    }
    margin=<ltrb 0 -6 0 1>
    renderTransformPivot=<0 0.75>
    shadowOffset=<0 0>
   }
   slot {class=BorderSlot, properties {padding=<ltrb 0 0 0 0>}}
  }
  TutorialHighlight {widgetClassPath=&TutorialHighlight}
 ]
}
